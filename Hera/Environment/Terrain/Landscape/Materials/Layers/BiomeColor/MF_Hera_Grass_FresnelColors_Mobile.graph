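# Mobile material function for the Hera landscape biome color layer that outputs a grass fresnel color from terrain masks.
# It lerps between exposed Mobile Grass Fresnel and Mobile GrassDesert Fresnel color parameters.
# The blend alpha is the max of QuadB red and the QuadD blue and alpha mask channels, so any of those masks shifts the tint toward the desert color.
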MaterialFunction /Hera/Environment/Terrain/Landscape/Materials/Layers/BiomeColor/MF_Hera_Grass_FresnelColors_Mobile
graph {nodes [
  n1 = VectorParameter() {group=Mobile Grass, parameterName=Mobile GrassDesert Fresnel, defaultValue=<rgba 0.0469 0.0457 0.0236 1>}
  n2 = FunctionInput() {bUsePreviewValueAsDefault=true, inputName=TerrainMasks_QuadB, inputType=FunctionInput_Vector4, sortPriority=4, previewValue=<1 1 1 1>}
  n3 = ComponentMask(None=@n2) {r=true}
  n4 = FunctionInput() {bUsePreviewValueAsDefault=true, inputName=TerrainMasks_QuadD, inputType=FunctionInput_Vector4, sortPriority=4, previewValue=<1 1 1 1>}
  n5 = Max(A=ComponentMask(None=@n4) {b=true}, B=ComponentMask(None=@n4) {a=true})
  FunctionOutput(None=LinearInterpolate(
    A=VectorParameter() {group=Mobile Grass, parameterName=Mobile Grass Fresnel, defaultValue=<rgba 0.1425 0.1927 0.051 1>}.RGB
    B=@n1.RGB
    Alpha=Max(A=@n3, B=@n5)
   )) {outputName=Grass Fresnel}
 ]}
