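# Landscape material function that tints the Hera sand layer's base color per biome, using masks decoded from packed terrain mask inputs.
# Chained lerps apply biome colors, with a static switch to a generic six-color chain for maps outside BR.
# Desert color blends five sediment colors through HeightLerp, while mobile derives bands from noise-distorted world height.
# The default path also adjusts desert specular; Island A/B tints exist but are unconnected.
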MaterialFunction /Hera/Environment/Terrain/Landscape/Materials/Layers/BiomeColor/MF_Hera_Sand_BiomeColor
graph {
 properties {libraryCategoriesText [FN-Landscape]}
 nodes [
  Constant4Vector() {constant=<rgba -1000 0.0009 -1300 0.0027>}
  Constant4Vector() {constant=<rgba 700 1.5 0.003 0.5>}
  Constant4Vector() {constant=<rgba 0.8 0.5925 0.2736 1>}
  Constant4Vector() {constant=<rgba 0.65 0.4467 0.2453 1>}
  Constant4Vector() {constant=<rgba 0.95 0.6424 0.3325 1>}
  Constant4Vector() {constant=<rgba 0.8 0.5082 0.24 1>}
  Constant4Vector() {constant=<rgba 1.1 0.7591 0.33 1>}
  n1 = NamedRerouteUsage()
  n2 = VectorParameter() {
   group=BiomeColors
   parameterName=SandColor_NarrowFlea
   channelNames {a=DesaturateBasecolor, b="", g="", r=""}
   defaultValue=<rgba 0.65 0.5851 0.474 1>
  }
  Desaturation(None=@n1, Fraction=@n2.DesaturateBasecolor)
  n3 = NamedRerouteUsage()
  n4 = VectorParameter() {
   group=BiomeColors
   parameterName=SandColor_IslandA
   channelNames {a=DesaturateBasecolor, b="", g="", r=""}
   defaultValue=<rgba 0.25 0.25 0.25 0.9>
  }
  Desaturation(None=@n3, Fraction=@n4.DesaturateBasecolor)
  n5 = NamedRerouteUsage()
  n6 = VectorParameter() {
   group=BiomeColors
   parameterName=SandColor_IslandB
   channelNames {a=DesaturateBasecolor, b="", g="", r=""}
   defaultValue=<rgba 0.25 0.25 0.25 0.9>
  }
  Desaturation(None=@n5, Fraction=@n6.DesaturateBasecolor)
  n7 = SetMaterialAttributes!(
   MaterialAttributes=NamedRerouteUsage()
   BaseColor=StaticSwitchParameter(True=NamedRerouteUsage(), False=NamedRerouteUsage()) {defaultValue=true, group=BiomeSettings, parameterName=UseBiomeColors}
  ) {attributeSetTypes [#1]}
  n8 = SetMaterialAttributes!(MaterialAttributes=@n7, Specular=NamedRerouteUsage()) {attributeSetTypes [#2]}
  FunctionOutput(None=ShadingPathSwitch(Default=StaticSwitchParameter(True=@n8, False=@n7) {defaultValue=true, group=BiomeSettings, parameterName=UseBiomeColors}, Mobile=@n7))
  FunctionOutput(None=SetMaterialAttributes!(MaterialAttributes=NamedRerouteUsage(), BaseColor=NamedRerouteUsage()) {attributeSetTypes [#1]}) {outputName=Result Mobile, sortPriority=1}
  n9 = VectorParameter() {group=BiomeColors, parameterName=SandColor_Grasslands, defaultValue=<rgba 0.65 0.4943 0.2275 1>}
  n10 = NamedRerouteUsage()
  n11 = NamedRerouteDeclaration(None=Saturate(None=Multiply(A=@n9.RGB, B=@n10))) {name=NoBiomeColor, nodeColor=<rgba 0.5 0 1 1>}
  n12 = Multiply(A=VectorParameter() {group=BiomeColors, parameterName=SandColor_RedWods, defaultValue=<rgba 0.4 0.4 0.4 1>}.RGB, B=NamedRerouteUsage())
  n13 = NamedRerouteUsage()
  n14 = Multiply(A=VectorParameter() {group=BiomeColors, parameterName=SandColor_Archipelago, defaultValue=<rgba 0.9647 0.8714 0.6038 1>}.RGB, B=@n13)
  n15 = VectorParameter() {group=BiomeColors, parameterName=Mobile_SandColor_Archipelago, defaultValue=<rgba 0.9647 0.9647 0.9647 1>}
  n16 = Multiply(A=@n15.RGB, B=@n13)
  n17 = LinearInterpolate(
   A=LinearInterpolate(A=@n11, B=Saturate(None=@n12), Alpha=NamedRerouteUsage())
   B=ShadingPathSwitch(Default=Saturate(None=@n14), Mobile=Saturate(None=@n16))
   Alpha=NamedRerouteUsage()
  )
  n18 = OneMinus(None=NamedRerouteUsage())
  n19 = ScalarParameter() {defaultValue=0.5, group=OpenDesert_MacroDetail, parameterName=OD_MD_LayerBlendContrast, sortPriority=501}
  n20 = LinearInterpolate(
   A=VectorParameter() {group=BiomeColors, parameterName=OD_MD_Color0, defaultValue=<rgba 1 0.5204 0.2161 1>}.RGB
   B=VectorParameter() {group=BiomeColors, parameterName=OD_MD_Color1, defaultValue=<rgba 0.65 0.2868 0.103 1>}.RGB
   Alpha=MaterialFunctionCall(Transition Phase=NamedRerouteUsage(), Height Texture=@n18, Contrast=@n19) {materialFunction=&HeightLerp}.Alpha
  )
  n21 = VectorParameter() {group=BiomeColors, parameterName=OD_MD_Color2, defaultValue=<rgba 0.75 0.369 0.1875 1>}
  n22 = MaterialFunctionCall(Transition Phase=NamedRerouteUsage(), Height Texture=@n18, Contrast=@n19) {materialFunction=&HeightLerp}
  n23 = NamedRerouteUsage()
  n24 = LinearInterpolate(
   A=LinearInterpolate(
    A=LinearInterpolate(A=@n20, B=@n21.RGB, Alpha=@n22.Alpha)
    B=VectorParameter() {group=BiomeColors, parameterName=OD_MD_Color3, defaultValue=<rgba 0.6 0.2545 0.1327 1>}.RGB
    Alpha=MaterialFunctionCall(Transition Phase=@n23, Height Texture=@n18, Contrast=@n19) {materialFunction=&HeightLerp}.Alpha
   )
   B=VectorParameter() {group=BiomeColors, parameterName=OD_MD_Color4, defaultValue=<rgba 0.8 0.3493 0.16 1>}.RGB
   Alpha=MaterialFunctionCall(Transition Phase=NamedRerouteUsage(), Height Texture=@n18, Contrast=@n19) {materialFunction=&HeightLerp}.Alpha
  )
  n25 = NamedRerouteUsage()
  n26 = Max(A=NamedRerouteUsage(), B=NamedRerouteUsage())
  n27 = VectorParameter() {group=BiomeColors, parameterName=Mobile OD Color4, defaultValue=<rgba 0.7656 0.4328 0.327 1>}
  n28 = Multiply(A=Add(A=NamedRerouteUsage(), B=NamedRerouteUsage()), B=LinearInterpolate(Alpha=NamedRerouteUsage()) {constA=0.6})
  n29 = Constant4Vector() {constant=<rgba -500 1.5 0.003 0.5>}
  n30 = Add(A=TruncateLWC!(None=WorldPosition().Z), B=Multiply(A=Add(A=@n28) {constB=-0.5}, B=@n29.R))
  n31 = Constant4Vector() {constant=<rgba -1000 0.0009 -1300 0.003>}
  n32 = Multiply(A=Add(A=@n30, B=@n31.R), B=@n31.G)
  n33 = Abs(None=Add(A=@n30, B=@n31.B))
  n34 = Saturate(None=Subtract(A=Multiply(A=@n33, B=@n31.A), B=@n29.G))
  n35 = LinearInterpolate(
   A=LinearInterpolate(
    A=VectorParameter() {group=BiomeColors, parameterName=Mobile OD Color1, defaultValue=<rgba 0.98 0.699 0.4645 1>}.RGB
    B=VectorParameter() {group=BiomeColors, parameterName=Mobile OD Color2, defaultValue=<rgba 1 0.536 0.3912 1>}.RGB
    Alpha=Saturate(None=@n32)
   )
   B=VectorParameter() {group=BiomeColors, parameterName=Mobile OD Color3, defaultValue=<rgba 0.4219 0.1083 0.0242 1>}.RGB
   Alpha=Multiply(A=OneMinus(None=@n34), B=@n29.A)
  )
  n36 = Saturate(None=Multiply(A=@n33, B=@n29.B))
  n37 = Multiply(A=NamedRerouteUsage(), B=Saturate(None=LinearInterpolate(A=@n27.RGB, B=@n35, Alpha=@n36)))
  n38 = LinearInterpolate(A=@n17, B=@n37, Alpha=@n26)
  n39 = ShadingPathSwitch(Default=LinearInterpolate(A=@n17, B=Multiply(A=@n24, B=@n25), Alpha=@n26), Mobile=@n38)
  n40 = Saturate(None=Multiply(A=VectorParameter() {group=BiomeColors, parameterName=SandColor_GoldenFall, defaultValue=<rgba 0.8 0.8 0.8 1>}.RGB, B=NamedRerouteUsage()))
  n41 = NamedRerouteUsage()
  n42 = Multiply(A=@n2.RGB, B=@n1)
  n43 = VectorParameter() {
   group=BiomeColors
   parameterName=SandColor_AxisMandate
   channelNames {a=DesaturateBasecolor, b="", g="", r=""}
   defaultValue=<rgba 0.8 0.572 0.28 1>
  }
  n44 = Desaturation(None=NamedRerouteUsage(), Fraction=@n43.DesaturateBasecolor)
  n45 = LinearInterpolate(
   A=LinearInterpolate(A=LinearInterpolate(A=@n39, B=@n40, Alpha=@n41), B=Saturate(None=@n42), Alpha=NamedRerouteUsage())
   B=Saturate(None=Multiply(A=@n43.RGB, B=@n44))
   Alpha=NamedRerouteUsage()
  )
  n46 = Multiply(A=@n4.RGB, B=@n3)
  LinearInterpolate(
   A=LinearInterpolate(A=@n45, B=Saturate(None=@n46), Alpha=NamedRerouteUsage())
   B=Saturate(None=Multiply(A=@n6.RGB, B=@n5))
   Alpha=NamedRerouteUsage()
  )
  n47 = VectorParameter() {group=BiomeColors, parameterName=SandColor_Chaparral, defaultValue=<rgba 0.6 0.5673 0.4854 1>}
  n48 = NamedRerouteUsage()
  LinearInterpolate(A=@n11, B=Saturate(None=Multiply(A=@n47.RGB, B=@n48)), Alpha=NamedRerouteUsage())
  n49 = MaterialFunctionCall(Float4=FunctionInput(Preview=Constant4Vector().RGBA) {bUsePreviewValueAsDefault=true, inputName=OD_MD_SedimentLayers, inputType=FunctionInput_Vector4, sortPriority=60}) {materialFunction=&BreakOutFloat4Components}
  NamedRerouteDeclaration(None=@n49.R) {name=OD_MD_Layer1, nodeColor=<rgba 0.5 0 1 1>}
  NamedRerouteDeclaration(None=@n49.G) {name=OD_MD_Layer2, nodeColor=<rgba 0.5 0 1 1>}
  n50 = StaticSwitchParameter(True=NamedRerouteUsage(), False=@n45) {desc=Used in maps outside of BR, group=GenericBiomeColors, parameterName=UseGenericBiome_ColorLerpChain, sortPriority=-1000}
  NamedRerouteDeclaration(None=ComponentMask(None=@n50) {b=true, g=true, r=true}) {name=OutputBasecolor, nodeColor=<rgba 0.5 0 1 1>}
  n51 = FunctionInput() {bUsePreviewValueAsDefault=true, inputName=GrassNoise, inputType=FunctionInput_Vector4, sortPriority=50, previewValue=<0 0 0 0>}
  NamedRerouteDeclaration(None=@n51) {name=GrassNoiseVec4, nodeColor=<rgba 0.5 0 1 1>}
  n52 = FunctionInput() {bUsePreviewValueAsDefault=true, inputName=TerrainMasks_QuadA, inputType=FunctionInput_Vector4, sortPriority=10}
  n53 = MaterialFunctionCall(Float4=@n52) {materialFunction=&BreakOutFloat4Components}
  NamedRerouteDeclaration(None=@n53.R) {name=BM_Generic01, nodeColor=<rgba 0.5 0 1 1>}
  n54 = GetMaterialAttributes(None=FunctionInput(Preview=SetMaterialAttributes!(MaterialAttributes)) {bUsePreviewValueAsDefault=true, inputName=Material, inputType=FunctionInput_MaterialAttributes}) {attributeGetTypes [
    #1
    #3
    #4
    #2
   ]}
  n55 = NamedRerouteDeclaration(None=@n54.BaseColor) {name=Input_BaseColor, nodeColor=<rgba 0.5 0 1 1>}
  NamedRerouteDeclaration(None=Desaturation(None=@n55, Fraction=Constant() {r=0.5})) {name=BaseColor_DesatHalf, nodeColor=<rgba 0.5 0 1 1>}
  NamedRerouteDeclaration(None=@n53.G) {name=BM_Generic02, nodeColor=<rgba 0.5 0 1 1>}
  n56 = Constant()
  n57 = Constant() {r=1.5}
  n58 = MaterialFunctionCall!(Coordinates, R_01_Scale, R_02_Scale, R_03_Scale, R_04_Scale, G_01_Scale, G_02_Scale, G_03_Scale, G_04_Scale) {materialFunction=/Game/Athena/Helios/Environments/Landscape/Materials/MaterialFunctions_Utilities/MF_VariationMasks2}
  NamedRerouteDeclaration(None=Saturate(None=LinearInterpolate(A=@n56, B=@n57, Alpha=@n58.G_01))) {name=Noise01, nodeColor=<rgba 0.5 0 1 1>}
  n59 = Constant() {r=1.5}
  n60 = Constant() {r=-9}
  NamedRerouteDeclaration(None=Saturate(None=LinearInterpolate(A=@n59, B=@n60, Alpha=@n58.G_02))) {name=Noise02, nodeColor=<rgba 0.5 0 1 1>}
  n61 = Constant() {r=-0.25}
  n62 = Constant() {r=1.5}
  n63 = NamedRerouteDeclaration(None=Saturate(None=LinearInterpolate(A=@n61, B=@n62, Alpha=@n58.G_03))) {name=Noise03_PreMult, nodeColor=<rgba 0.5 0 1 1>}
  n64 = Constant() {r=0.65}
  NamedRerouteDeclaration(None=Saturate(None=Multiply(A=@n63, B=@n64))) {name=Noise03, nodeColor=<rgba 0.5 0 1 1>}
  NamedRerouteDeclaration(None=@n49.B) {name=OD_MD_Layer3, nodeColor=<rgba 0.5 0 1 1>}
  n65 = Constant() {r=4}
  NamedRerouteDeclaration(None=Saturate(None=Multiply(A=@n58.G_04, B=@n65))) {name=Noise04, nodeColor=<rgba 0.5 0 1 1>}
  NamedRerouteDeclaration(None=@n49.A) {name=OD_MD_Layer4, nodeColor=<rgba 0.5 0 1 1>}
  NamedRerouteDeclaration(None=NamedRerouteDeclaration(None=@n54.Roughness) {name=Input_Roughness, nodeColor=<rgba 0.5 0 1 1>}) {name=OutputRoughness, nodeColor=<rgba 0.5 0 1 1>}
  n66 = FunctionInput() {bUsePreviewValueAsDefault=true, inputName=TerrainMasks_QuadB, inputType=FunctionInput_Vector4, sortPriority=20}
  n67 = MaterialFunctionCall!(
   TerrainMasks_QuadA=@n52
   TerrainMasks_QuadB=@n66
   TerrainMasks_QuadC=FunctionInput() {bUsePreviewValueAsDefault=true, inputName=TerrainMasks_QuadC, inputType=FunctionInput_Vector4, sortPriority=30}
   TerrainMasks_QuadD=FunctionInput() {bUsePreviewValueAsDefault=true, inputName=TerrainMasks_QuadD, inputType=FunctionInput_Vector4, sortPriority=40}
  ) {materialFunction=/BRCoreMaterials/Hera/Terrain/BiomeMask/MF_HeraBiomeMasks_Isolated}
  NamedRerouteDeclaration(None=@n67.00_Chaparral) {name=BM_Empty, nodeColor=<rgba 0.5 0 1 1>}
  NamedRerouteDeclaration(None=@n67.01_RedWoods) {name=BM_RedWoods, nodeColor=<rgba 0.5 0 1 1>}
  NamedRerouteDeclaration(None=@n67.02_Archipelago) {name=BM_Archipelago, nodeColor=<rgba 0.5 0 1 1>}
  NamedRerouteDeclaration(None=@n67.03_Desert) {name=BM_Desert, nodeColor=<rgba 0.5 0 1 1>}
  NamedRerouteDeclaration(None=@n67.04_GoldenFall) {name=BM_GoldenFall, nodeColor=<rgba 0.5 0 1 1>}
  NamedRerouteDeclaration(None=@n53.B) {name=BM_Generic03, nodeColor=<rgba 0.5 0 1 1>}
  NamedRerouteDeclaration(None=@n53.A) {name=BM_Generic04, nodeColor=<rgba 0.5 0 1 1>}
  n68 = MaterialFunctionCall(Float4=@n66) {materialFunction=&BreakOutFloat4Components}
  NamedRerouteDeclaration(None=@n68.R) {name=BM_Generic05, nodeColor=<rgba 0.5 0 1 1>}
  NamedRerouteDeclaration(None=@n68.G) {name=BM_Generic06, nodeColor=<rgba 0.5 0 1 1>}
  NamedRerouteDeclaration(None=@n67.06_NarrowFlea) {name=BM_NarrowFlea, nodeColor=<rgba 0.5 0 1 1>}
  NamedRerouteDeclaration(None=@n67.07_AxisMandate) {name=BM_AxisMandate, nodeColor=<rgba 0.5 0 1 1>}
  n69 = Saturate(None=Multiply(
    A=VectorParameter() {group=GenericBiomeColors, parameterName=GBC_Sand_00, sortPriority=0, defaultValue=<rgba 1 1 1 1>}.RGB
    B=NamedRerouteUsage()
   ))
  n70 = Saturate(None=Multiply(
    A=VectorParameter() {group=GenericBiomeColors, parameterName=GBC_Sand_01, sortPriority=10, defaultValue=<rgba 1 1 1 1>}.RGB
    B=NamedRerouteUsage()
   ))
  n71 = NamedRerouteUsage()
  n72 = Multiply(
   A=VectorParameter() {group=GenericBiomeColors, parameterName=GBC_Sand_02, sortPriority=20, defaultValue=<rgba 1 1 1 1>}.RGB
   B=NamedRerouteUsage()
  )
  n73 = VectorParameter() {group=GenericBiomeColors, parameterName=GBC_Sand_03, sortPriority=30, defaultValue=<rgba 1 1 1 1>}
  n74 = NamedRerouteUsage()
  n75 = LinearInterpolate(
   A=LinearInterpolate(A=LinearInterpolate(A=@n69, B=@n70, Alpha=@n71), B=Saturate(None=@n72), Alpha=NamedRerouteUsage())
   B=Saturate(None=Multiply(A=@n73.RGB, B=@n74))
   Alpha=NamedRerouteUsage()
  )
  n76 = Saturate(None=Multiply(
    A=VectorParameter() {group=GenericBiomeColors, parameterName=GBC_Sand_04, sortPriority=40, defaultValue=<rgba 1 1 1 1>}.RGB
    B=NamedRerouteUsage()
   ))
  n77 = NamedRerouteUsage()
  n78 = Multiply(
   A=VectorParameter() {group=GenericBiomeColors, parameterName=GBC_Sand_05, sortPriority=50, defaultValue=<rgba 1 1 1 1>}.RGB
   B=NamedRerouteUsage()
  )
  NamedRerouteDeclaration(None=LinearInterpolate(A=LinearInterpolate(A=@n75, B=@n76, Alpha=@n77), B=Saturate(None=@n78), Alpha=NamedRerouteUsage())) {name=OutputGeneric, nodeColor=<rgba 0.5 0 1 1>}
  NamedRerouteDeclaration(None=@n54.Specular) {name=Input_Spec, nodeColor=<rgba 0.5 0 1 1>}
  NamedRerouteDeclaration(None=@n67.08_IslandA) {name=BM_IslandA, nodeColor=<rgba 0.5 0 1 1>}
  NamedRerouteDeclaration(None=@n67.09_IslandB) {name=BM_IslandB, nodeColor=<rgba 0.5 0 1 1>}
  n79 = NamedRerouteUsage()
  n80 = Multiply(A=@n79, B=ScalarParameter() {defaultValue=1, group=BiomeSettings, parameterName=Desert_SandSpecMult})
  NamedRerouteDeclaration(None=LinearInterpolate(A=@n79, B=Saturate(None=@n80), Alpha=NamedRerouteUsage())) {name=OutputSpecular, nodeColor=<rgba 0.5 0 1 1>}
  NamedRerouteDeclaration(None=Desaturation(None=@n55, Fraction=Constant() {r=1})) {name=BaseColorDesat, nodeColor=<rgba 0.5 0 1 1>}
  NamedRerouteDeclaration(None=@n54.MaterialAttributes) {name=InputMaterial, nodeColor=<rgba 0.5 0 1 1>}
  NamedRerouteDeclaration(None=@n54.OpacityMask) {name=Heightmap, nodeColor=<rgba 0.5 0 1 1>}
  NamedRerouteDeclaration(None=ComponentMask(None=Constant3Vector() {constant=<rgba 1 0 0 0>}) {b=true, g=true, r=true}) {name=OutputBasecolorMobile, nodeColor=<rgba 0.5 0 1 1>}
  NamedRerouteUsage()
  NamedRerouteUsage()
  Saturate(None=Add(A=@n37, B=VectorParameter() {group=BiomeColors, parameterName=Mobile OD Color5, defaultValue=<rgba 0 0 0 1>}.RGB))
  ScalarParameter() {defaultValue=-1000, group=BiomeColors, parameterName=Mobile OD MaskAdd1}
  ScalarParameter() {defaultValue=0.0009, group=BiomeColors, parameterName=Mobile OD MaskMul1}
  ScalarParameter() {defaultValue=700, group=BiomeColors, parameterName=Mobile OD MaskDistortion}
  ScalarParameter() {defaultValue=-1300, group=BiomeColors, parameterName=Mobile OD MaskAdd2}
  ScalarParameter() {defaultValue=0.0027, group=BiomeColors, parameterName=Mobile OD MaskMul2}
  ScalarParameter() {defaultValue=1.5, group=BiomeColors, parameterName=Mobile OD MaskSub2}
  ScalarParameter() {defaultValue=0.5, group=BiomeColors, parameterName=Mobile OD MaskOpc2}
  ScalarParameter() {defaultValue=0.003, group=BiomeColors, parameterName=Mobile OD MaskMul3}
  SetMaterialAttributes!(
   MaterialAttributes
   BaseColor
   Roughness=StaticSwitchParameter(True=NamedRerouteUsage(), False=NamedRerouteUsage()) {defaultValue=true, group=BiomeSettings, parameterName=UseBiomeColors}
  ) {attributeSetTypes [#1, #4]}
  StaticSwitch(True=@n38, False=@n17, Value=StaticBoolParameter() {group=Mobile Sand, parameterName=Mobile Sand UseDesertColors})
  VectorParameter() {
   group=BiomeColors
   parameterName=Mobile OD Settings1
   channelNames {a=Mobile OD MaskOpc2, b=Mobile OD MaskMul3, g=Mobile OD MaskSub2, r=Mobile OD MaskDistortion}
   defaultValue=<rgba -500 1.5 0.003 0.5>
  }
  VectorParameter() {
   group=BiomeColors
   parameterName=Mobile OD Settings 2
   channelNames {a=Mobile OD MaskMul2, b=Mobile OD MaskAdd2, g=Mobile OD MaskMul1, r=Mobile OD MaskAdd1}
   defaultValue=<rgba -1000 0.0009 -1300 0.003>
  }
 ]
}
