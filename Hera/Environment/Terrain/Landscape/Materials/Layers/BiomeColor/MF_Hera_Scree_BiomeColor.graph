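# Landscape material function that tints the scree layer's base color per biome, toggled by the UseBiomeColors static switch.
# Isolated biome masks from the terrain mask quads drive a lerp chain of saturated color-parameter multiplies, with mobile-specific tints for Chaparral and Desert.
# UseGenericBiome_ColorLerpChain swaps in a six-color generic chain for maps outside BR.
# Outputs material attributes with BaseColor replaced, plus a separate Result Mobile output.
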MaterialFunction /Hera/Environment/Terrain/Landscape/Materials/Layers/BiomeColor/MF_Hera_Scree_BiomeColor
graph {
 properties {libraryCategoriesText [FN-Landscape]}
 nodes [
  Constant4Vector() {constant=<rgba 0.55 0.4462 0.33 1>}
  n1 = NamedRerouteUsage()
  n2 = NamedRerouteUsage()
  FunctionOutput(None=SetMaterialAttributes!(
    MaterialAttributes=NamedRerouteUsage()
    BaseColor=StaticSwitchParameter(True=@n1, False=@n2) {defaultValue=true, group=BiomeSettings, parameterName=UseBiomeColors}
   ) {attributeSetTypes [#1]})
  n3 = NamedRerouteUsage()
  n4 = SetMaterialAttributes(MaterialAttributes=@n3, BaseColor=NamedRerouteUsage()) {attributeSetTypes [#1]}
  FunctionOutput(None=@n4) {outputName=Result Mobile, sortPriority=1}
  n5 = NamedRerouteUsage()
  Max(A=@n5, B=NamedRerouteUsage())
  n6 = FunctionInput() {bUsePreviewValueAsDefault=true, inputName=TerrainMasks_QuadA, inputType=FunctionInput_Vector4, sortPriority=10}
  n7 = FunctionInput() {bUsePreviewValueAsDefault=true, inputName=TerrainMasks_QuadB, inputType=FunctionInput_Vector4, sortPriority=20}
  n8 = MaterialFunctionCall!(
   TerrainMasks_QuadA=@n6
   TerrainMasks_QuadB=@n7
   TerrainMasks_QuadC=FunctionInput() {bUsePreviewValueAsDefault=true, inputName=TerrainMasks_QuadC, inputType=FunctionInput_Vector4, sortPriority=30}
   TerrainMasks_QuadD=FunctionInput() {bUsePreviewValueAsDefault=true, inputName=TerrainMasks_QuadD, inputType=FunctionInput_Vector4, sortPriority=40}
  ) {materialFunction=/BRCoreMaterials/Hera/Terrain/BiomeMask/MF_HeraBiomeMasks_Isolated}
  NamedRerouteDeclaration(None=@n8.00_Chaparral) {name=BM_Empty, nodeColor=<rgba 0.5 0 1 1>}
  NamedRerouteDeclaration(None=@n8.01_RedWoods) {name=BM_Redwoods, nodeColor=<rgba 0.5 0 1 1>}
  n9 = Saturate(None=Multiply(
    A=VectorParameter() {group=BiomeColors, parameterName=ScreeColor_Grasslands, defaultValue=<rgba 1.1 1.0266 0.762 1>}.RGB
    B=NamedRerouteUsage()
   ))
  n10 = NamedRerouteUsage()
  n11 = Saturate(None=Multiply(A=VectorParameter() {group=BiomeColors, parameterName=ScreeColor_Chaparral, defaultValue=<rgba 1.3 1.1744 0.7042 1>}.RGB, B=@n10))
  n12 = Saturate(None=Multiply(A=VectorParameter() {group=BiomeColors, parameterName=Mobile_ScreeColor_Chaparral, defaultValue=<rgba 1 0.9036 0.75 1>}.RGB, B=@n10))
  n13 = VectorParameter() {group=BiomeColors, parameterName=ScreeColor_RedWods, defaultValue=<rgba 0.55 0.4462 0.33 1>}
  n14 = NamedRerouteUsage()
  n15 = LinearInterpolate(
   A=LinearInterpolate(
    A=NamedRerouteDeclaration(None=@n9) {name=NoBiomeColor, nodeColor=<rgba 0.5 0 1 1>}
    B=ShadingPathSwitch(Default=@n11, Mobile=@n12)
    Alpha=NamedRerouteUsage()
   )
   B=Saturate(None=Multiply(A=@n13.RGB, B=@n14))
   Alpha=NamedRerouteUsage()
  )
  n16 = Saturate(None=Multiply(
    A=VectorParameter() {group=BiomeColors, parameterName=ScreeColor_Archipelago, defaultValue=<rgba 0.8333 0.6593 0.4644 1>}.RGB
    B=NamedRerouteUsage()
   ))
  n17 = NamedRerouteUsage()
  n18 = VectorParameter() {group=BiomeColors, parameterName=ScreeColor_Desert, defaultValue=<rgba 1.2 0.5986 0.3424 1>}
  n19 = ShadingPathSwitch(
   Default=Multiply(A=@n17, B=@n18.RGB)
   Mobile=Multiply(
    A=NamedRerouteUsage()
    B=VectorParameter() {group=BiomeColors, parameterName=Mobile ScreeColor_Desert, defaultValue=<rgba 1.2 0.629 0.36 1>}.RGB
   )
  )
  n20 = VectorParameter() {group=BiomeColors, parameterName=ScreeColor_GoldenFall, defaultValue=<rgba 0.9 0.7826 0.72 1>}
  n21 = NamedRerouteUsage()
  n22 = LinearInterpolate(
   A=LinearInterpolate(A=LinearInterpolate(A=@n15, B=@n16, Alpha=@n5), B=Saturate(None=@n19), Alpha=NamedRerouteUsage())
   B=Saturate(None=Multiply(A=@n20.RGB, B=@n21))
   Alpha=NamedRerouteUsage()
  )
  n23 = Saturate(None=Multiply(
    A=VectorParameter() {group=BiomeColors, parameterName=ScreeColor_NarrowFlea, defaultValue=<rgba 0.9 0.7826 0.72 1>}.RGB
    B=NamedRerouteUsage()
   ))
  n24 = NamedRerouteUsage()
  n25 = Multiply(
   A=VectorParameter() {group=BiomeColors, parameterName=ScreeColor_AxisMandate, defaultValue=<rgba 1.5 1.2831 0.75 1>}.RGB
   B=NamedRerouteUsage()
  )
  n26 = StaticSwitchParameter(
   True=NamedRerouteUsage()
   False=LinearInterpolate(A=LinearInterpolate(A=@n22, B=@n23, Alpha=@n24), B=Saturate(None=@n25), Alpha=NamedRerouteUsage())
  ) {desc=Used in maps outside of BR, group=GenericBiomeColors, parameterName=UseGenericBiome_ColorLerpChain, sortPriority=-1000}
  NamedRerouteDeclaration(None=ComponentMask(None=@n26) {b=true, g=true, r=true}) {name=OutputBasecolor, nodeColor=<rgba 0.5 0 1 1>}
  n27 = FunctionInput() {bUsePreviewValueAsDefault=true, inputName=GrassNoise, inputType=FunctionInput_Vector4, sortPriority=50, previewValue=<0 0 0 0>}
  NamedRerouteDeclaration(None=@n27) {name=GrassNoiseVec4, nodeColor=<rgba 0.5 0 1 1>}
  NamedRerouteDeclaration(None=@n8.04_GoldenFall) {name=BM_GoldenFall, nodeColor=<rgba 0.5 0 1 1>}
  n28 = GetMaterialAttributes(None=FunctionInput(Preview=SetMaterialAttributes!(MaterialAttributes)) {bUsePreviewValueAsDefault=true, inputName=Material, inputType=FunctionInput_MaterialAttributes}) {attributeGetTypes [#1, #2]}
  n29 = NamedRerouteDeclaration(None=@n28.BaseColor) {name=Input_BaseColor, nodeColor=<rgba 0.5 0 1 1>}
  NamedRerouteDeclaration(None=Desaturation(None=@n29, Fraction=Constant() {r=0.5})) {name=BaseColor_DesatHalf, nodeColor=<rgba 0.5 0 1 1>}
  n30 = MaterialFunctionCall(Float4=@n6) {materialFunction=&BreakOutFloat4Components}
  NamedRerouteDeclaration(None=@n30.G) {name=BM_Generic02, nodeColor=<rgba 0.5 0 1 1>}
  n31 = Constant()
  n32 = Constant() {r=1.5}
  n33 = MaterialFunctionCall!(Coordinates, R_01_Scale, R_02_Scale, R_03_Scale, R_04_Scale, G_01_Scale, G_02_Scale, G_03_Scale, G_04_Scale) {materialFunction=/Game/Athena/Helios/Environments/Landscape/Materials/MaterialFunctions_Utilities/MF_VariationMasks2}
  NamedRerouteDeclaration(None=Saturate(None=LinearInterpolate(A=@n31, B=@n32, Alpha=@n33.G_01))) {name=Noise01, nodeColor=<rgba 0.5 0 1 1>}
  n34 = Constant() {r=1.5}
  n35 = Constant() {r=-9}
  NamedRerouteDeclaration(None=Saturate(None=LinearInterpolate(A=@n34, B=@n35, Alpha=@n33.G_02))) {name=Noise02, nodeColor=<rgba 0.5 0 1 1>}
  n36 = Constant() {r=-0.25}
  n37 = Constant() {r=1.5}
  n38 = NamedRerouteDeclaration(None=Saturate(None=LinearInterpolate(A=@n36, B=@n37, Alpha=@n33.G_03))) {name=Noise03_PreMult, nodeColor=<rgba 0.5 0 1 1>}
  n39 = Constant() {r=0.65}
  NamedRerouteDeclaration(None=Saturate(None=Multiply(A=@n38, B=@n39))) {name=Noise03, nodeColor=<rgba 0.5 0 1 1>}
  NamedRerouteDeclaration(None=@n8.02_Archipelago) {name=BM_Archipelago, nodeColor=<rgba 0.5 0 1 1>}
  n40 = Constant() {r=4}
  NamedRerouteDeclaration(None=Saturate(None=Multiply(A=@n33.G_04, B=@n40))) {name=Noise04, nodeColor=<rgba 0.5 0 1 1>}
  NamedRerouteDeclaration(None=@n30.B) {name=BM_Generic03, nodeColor=<rgba 0.5 0 1 1>}
  NamedRerouteDeclaration(None=@n30.A) {name=BM_Generic04, nodeColor=<rgba 0.5 0 1 1>}
  n41 = MaterialFunctionCall(Float4=@n7) {materialFunction=&BreakOutFloat4Components}
  NamedRerouteDeclaration(None=@n41.R) {name=BM_Generic05, nodeColor=<rgba 0.5 0 1 1>}
  NamedRerouteDeclaration(None=@n41.G) {name=BM_Generic06, nodeColor=<rgba 0.5 0 1 1>}
  NamedRerouteDeclaration(None=@n8.06_NarrowFlea) {name=BM_NarrowFlea, nodeColor=<rgba 0.5 0 1 1>}
  NamedRerouteDeclaration(None=@n8.07_AxisMandate) {name=BM_AxisMandate, nodeColor=<rgba 0.5 0 1 1>}
  n42 = Saturate(None=Multiply(
    A=VectorParameter() {group=GenericBiomeColors, parameterName=GBC_Scree_00, sortPriority=0, defaultValue=<rgba 1 1 1 1>}.RGB
    B=NamedRerouteUsage()
   ))
  n43 = VectorParameter() {group=GenericBiomeColors, parameterName=GBC_Scree_01, sortPriority=10, defaultValue=<rgba 1 1 1 1>}
  n44 = Saturate(None=Multiply(A=@n43.RGB, B=NamedRerouteUsage()))
  n45 = NamedRerouteUsage()
  n46 = VectorParameter() {group=GenericBiomeColors, parameterName=GBC_Scree_02, sortPriority=20, defaultValue=<rgba 1 1 1 1>}
  n47 = Multiply(A=@n46.RGB, B=NamedRerouteUsage())
  n48 = VectorParameter() {group=GenericBiomeColors, parameterName=GBC_Scree_03, sortPriority=30, defaultValue=<rgba 1 1 1 1>}
  n49 = NamedRerouteUsage()
  n50 = LinearInterpolate(
   A=LinearInterpolate(A=LinearInterpolate(A=@n42, B=@n44, Alpha=@n45), B=Saturate(None=@n47), Alpha=NamedRerouteUsage())
   B=Saturate(None=Multiply(A=@n48.RGB, B=@n49))
   Alpha=NamedRerouteUsage()
  )
  n51 = VectorParameter() {group=GenericBiomeColors, parameterName=GBC_Scree_04, sortPriority=40, defaultValue=<rgba 1 1 1 1>}
  n52 = Saturate(None=Multiply(A=@n51.RGB, B=NamedRerouteUsage()))
  n53 = NamedRerouteUsage()
  n54 = VectorParameter() {group=GenericBiomeColors, parameterName=GBC_Scree_05, sortPriority=50, defaultValue=<rgba 1 1 1 1>}
  n55 = Multiply(A=@n54.RGB, B=NamedRerouteUsage())
  NamedRerouteDeclaration(None=LinearInterpolate(A=LinearInterpolate(A=@n50, B=@n52, Alpha=@n53), B=Saturate(None=@n55), Alpha=NamedRerouteUsage())) {name=OutputGeneric, nodeColor=<rgba 0.5 0 1 1>}
  NamedRerouteDeclaration(None=@n8.03_Desert) {name=BM_Desert, nodeColor=<rgba 0.5 0 1 1>}
  NamedRerouteDeclaration(None=@n8.08_IslandA) {name=BM_IslandA, nodeColor=<rgba 0.5 0 1 1>}
  NamedRerouteDeclaration(None=@n8.09_IslandB) {name=BM_IslandB, nodeColor=<rgba 0.5 0 1 1>}
  NamedRerouteDeclaration(None=@n30.R) {name=BM_Generic01, nodeColor=<rgba 0.5 0 1 1>}
  NamedRerouteDeclaration(None=Desaturation(None=@n29, Fraction=Constant() {r=1})) {name=BaseColorDesat, nodeColor=<rgba 0.5 0 1 1>}
  NamedRerouteDeclaration(None=@n28.MaterialAttributes) {name=InputMaterial, nodeColor=<rgba 0.5 0 1 1>}
  NamedRerouteDeclaration(None=@n28.OpacityMask) {name=Heightmap, nodeColor=<rgba 0.5 0 1 1>}
  NamedRerouteDeclaration(None=ComponentMask(None=Constant3Vector() {constant=<rgba 1 0 0 0>}) {b=true, g=true, r=true}) {name=OutputBasecolorMobile, nodeColor=<rgba 0.5 0 1 1>}
  Power(Base=@n17, Exp=@n18.A)
  StaticSwitchParameter(True=@n4, False=@n3) {defaultValue=true, parameterName=UseHeliosBIomeColors}
 ]
}
