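# Landscape material function that tints the cliff layer's base color per biome, with a static switch for biome coloring.
# Terrain mask quads feed MF_HeraBiomeMasks_Isolated for named biome masks, driving chained lerps between saturated color-parameter multiplies.
# A second switch selects a generic six-color chain for maps outside BR, with raw quad channels exposed as generic masks.
# Shading path switches provide separate mobile colors, some using alpha as desaturation.
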MaterialFunction /Hera/Environment/Terrain/Landscape/Materials/Layers/BiomeColor/MF_Hera_Cliff_BiomeColor
graph {
 properties {libraryCategoriesText [FN-Landscape]}
 nodes [
  Constant4Vector() {desc=Light Block, constant=<rgba 1.462 0.8136 0.3229 1>}
  Constant4Vector() {desc=Dark Block, constant=<rgba 0.83 0.4745 0.1725 1>}
  n1 = NamedRerouteUsage()
  n2 = NamedRerouteUsage()
  FunctionOutput(None=SetMaterialAttributes!(
    MaterialAttributes=NamedRerouteUsage()
    BaseColor=StaticSwitchParameter(True=@n1, False=@n2) {defaultValue=true, group=BiomeSettings, parameterName=UseBiomeColors}
   ) {attributeSetTypes [#1]})
  n3 = Saturate(None=Multiply(A=VectorParameter() {group=BiomeColors, parameterName=CliffColor_Chaparral, defaultValue=<rgba 1.5 1.5 1.5 1>}.RGB, B=NamedRerouteUsage()))
  n4 = Saturate(None=Multiply(A=VectorParameter() {group=BiomeColors, parameterName=CliffColor_Grasslands, defaultValue=<rgba 1 1 1 1>}.RGB, B=NamedRerouteUsage()))
  LinearInterpolate(A=@n4, B=@n3, Alpha=NamedRerouteUsage())
  n5 = VectorParameter() {group=BiomeColors, parameterName=Mobile_CliffColor_Grasslands, channelNames {a=Desaturate, b="", g="", r=""}, defaultValue=<rgba 1 1 1 0>}
  n6 = NamedRerouteUsage()
  n7 = Saturate(None=Multiply(A=@n5.RGB, B=Desaturation(None=@n6, Fraction=@n5.Desaturate)))
  n8 = VectorParameter() {group=BiomeColors, parameterName=Mobile_CliffColor_Chaparral, defaultValue=<rgba 1 1 1 0>}
  n9 = NamedRerouteUsage()
  n10 = Saturate(None=Multiply(A=@n8.RGB, B=Desaturation(None=@n9, Fraction=@n8.A)))
  LinearInterpolate(A=@n7, B=@n10, Alpha=NamedRerouteUsage())
  n11 = NamedRerouteUsage()
  Max(A=@n11, B=NamedRerouteUsage())
  n12 = NamedRerouteUsage()
  Max(A=@n12, B=NamedRerouteUsage())
  NamedRerouteDeclaration(None=NamedRerouteUsage()) {name=NoBiomeColor, nodeColor=<rgba 0.5 0 1 1>}
  NamedRerouteDeclaration(None=FunctionInput() {bUsePreviewValueAsDefault=true, inputName=MossMask, inputType=FunctionInput_Scalar, sortPriority=100}) {name=MossMask, nodeColor=<rgba 0.5 0 1 1>}
  n13 = FunctionInput() {bUsePreviewValueAsDefault=true, inputName=TerrainMasks_QuadA, inputType=FunctionInput_Vector4, sortPriority=10}
  n14 = FunctionInput() {bUsePreviewValueAsDefault=true, inputName=TerrainMasks_QuadB, inputType=FunctionInput_Vector4, sortPriority=20}
  n15 = MaterialFunctionCall!(
   TerrainMasks_QuadA=@n13
   TerrainMasks_QuadB=@n14
   TerrainMasks_QuadC=FunctionInput() {bUsePreviewValueAsDefault=true, inputName=TerrainMasks_QuadC, inputType=FunctionInput_Vector4, sortPriority=30}
   TerrainMasks_QuadD=FunctionInput() {bUsePreviewValueAsDefault=true, inputName=TerrainMasks_QuadD, inputType=FunctionInput_Vector4, sortPriority=40}
  ) {materialFunction=/BRCoreMaterials/Hera/Terrain/BiomeMask/MF_HeraBiomeMasks_Isolated}
  NamedRerouteDeclaration(None=@n15.07_AxisMandate) {name=BM_AxisMandate, nodeColor=<rgba 0.5 0 1 1>}
  n16 = Multiply(A=VectorParameter() {group=BiomeColors, parameterName=CliffColor_RedWoods, defaultValue=<rgba 0.6 0.426 0.3 1>}.RGB, B=NamedRerouteUsage())
  n17 = VectorParameter() {group=BiomeColors, parameterName=CliffColor_Archipelago, defaultValue=<rgba 0.3916 0.3271 0.3133 1>}
  n18 = NamedRerouteUsage()
  n19 = LinearInterpolate(A=LinearInterpolate(A=@n4, B=Saturate(None=@n16), Alpha=NamedRerouteUsage()), B=Saturate(None=Multiply(A=@n17.RGB, B=@n18)), Alpha=@n12)
  n20 = Saturate(None=Multiply(A=VectorParameter() {group=BiomeColors, parameterName=CliffColor_Desert, defaultValue=<rgba 1 1 1 1>}.RGB, B=NamedRerouteUsage()))
  n21 = NamedRerouteUsage()
  n22 = Multiply(
   A=VectorParameter() {group=BiomeColors, parameterName=CliffColor_GoldenFall, defaultValue=<rgba 1.2 0.9599 0.84 1>}.RGB
   B=NamedRerouteUsage()
  )
  n23 = VectorParameter() {group=BiomeColors, parameterName=CliffColor_NarrowFlea, defaultValue=<rgba 1.462 0.8136 0.3229 1>}
  n24 = NamedRerouteUsage()
  n25 = LinearInterpolate(
   A=LinearInterpolate(A=LinearInterpolate(A=@n19, B=@n20, Alpha=@n21), B=Saturate(None=@n22), Alpha=NamedRerouteUsage())
   B=Saturate(None=Multiply(A=@n23.RGB, B=@n24))
   Alpha=NamedRerouteUsage()
  )
  n26 = Saturate(None=Multiply(
    A=VectorParameter() {group=BiomeColors, parameterName=CliffColor_AxisMandate, defaultValue=<rgba 1.25 1.0654 0.625 1>}.RGB
    B=NamedRerouteUsage()
   ))
  n27 = NamedRerouteUsage()
  n28 = VectorParameter() {group=BiomeColors, parameterName=Mobile_CliffColor_RedWoods, defaultValue=<rgba 0.6 0.426 0.3 0>}
  n29 = Multiply(A=@n28.RGB, B=Desaturation(None=NamedRerouteUsage(), Fraction=@n28.A))
  n30 = VectorParameter() {
   desc="This is used on multiple texture sets && can have different values in different MIs"
   group=BiomeColors
   parameterName=Mobile_CliffColor_Archipelago
   channelNames {a=Desaturate, b="", g="", r=""}
   defaultValue=<rgba 0.3916 0.2623 0.2346 1>
  }
  n31 = Desaturation(None=NamedRerouteUsage(), Fraction=@n30.Desaturate)
  n32 = LinearInterpolate(A=LinearInterpolate(A=@n7, B=Saturate(None=@n29), Alpha=NamedRerouteUsage()), B=Saturate(None=Multiply(A=@n30.RGB, B=@n31)), Alpha=@n11)
  n33 = Saturate(None=Multiply(A=VectorParameter() {group=BiomeColors, parameterName=Mobile_CliffColor_Desert, defaultValue=<rgba 1 1 1 1>}.RGB, B=NamedRerouteUsage()))
  n34 = NamedRerouteUsage()
  n35 = Multiply(
   A=VectorParameter() {group=BiomeColors, parameterName=Mobile_CliffColor_GoldenFall, defaultValue=<rgba 1.15 1.1103 1.035 0>}.RGB
   B=NamedRerouteUsage()
  )
  n36 = VectorParameter() {group=BiomeColors, parameterName=Mobile_CliffColor_NarrowFlea, defaultValue=<rgba 1.1652 0.6601 0.1925 1>}
  n37 = NamedRerouteUsage()
  n38 = LinearInterpolate(
   A=LinearInterpolate(A=LinearInterpolate(A=@n32, B=@n33, Alpha=@n34), B=Saturate(None=@n35), Alpha=NamedRerouteUsage())
   B=Saturate(None=Multiply(A=@n36.RGB, B=@n37))
   Alpha=NamedRerouteUsage()
  )
  n39 = VectorParameter() {group=BiomeColors, parameterName=Mobile_CliffColor_AxisMandate, defaultValue=<rgba 1.75 1.4464 0.9625 1>}
  n40 = Saturate(None=Multiply(A=@n39.RGB, B=NamedRerouteUsage()))
  n41 = NamedRerouteUsage()
  n42 = StaticSwitchParameter(
   True=ShadingPathSwitch(Default=NamedRerouteUsage(), Mobile=NamedRerouteUsage())
   False=ShadingPathSwitch(Default=LinearInterpolate(A=@n25, B=@n26, Alpha=@n27), Mobile=LinearInterpolate(A=@n38, B=@n40, Alpha=@n41))
  ) {desc=Used in maps outside of BR, group=GenericBiomeColors, parameterName=UseGenericBiome_ColorLerpChain, sortPriority=-1000}
  NamedRerouteDeclaration(None=ComponentMask(None=@n42) {b=true, g=true, r=true}) {name=OutputBasecolor, nodeColor=<rgba 0.5 0 1 1>}
  n43 = GetMaterialAttributes(None=FunctionInput(Preview=SetMaterialAttributes!(MaterialAttributes)) {bUsePreviewValueAsDefault=true, inputName=Material, inputType=FunctionInput_MaterialAttributes}) {attributeGetTypes [#1]}
  NamedRerouteDeclaration(None=Desaturation(
    None=NamedRerouteDeclaration(None=@n43.BaseColor) {name=Input_BaseColor, nodeColor=<rgba 0.5 0 1 1>}
    Fraction=ScalarParameter() {defaultValue=1, group=BiomeColors, parameterName=Cliff_BiomeDesat}
   )) {name=BaseColorDesat, nodeColor=<rgba 0.5 0 1 1>}
  NamedRerouteDeclaration(None=@n43.MaterialAttributes) {name=InputMaterial, nodeColor=<rgba 0.5 0 1 1>}
  NamedRerouteDeclaration(None=@n15.03_Desert) {name=BM_Desert, nodeColor=<rgba 0.5 0 1 1>}
  n44 = MaterialFunctionCall(Float4=@n13) {materialFunction=&BreakOutFloat4Components}
  NamedRerouteDeclaration(None=@n44.G) {name=BM_Generic02, nodeColor=<rgba 0.5 0 1 1>}
  NamedRerouteDeclaration(None=@n15.04_GoldenFall) {name=BM_GoldenFall, nodeColor=<rgba 0.5 0 1 1>}
  NamedRerouteDeclaration(None=@n44.R) {name=BM_Generic01, nodeColor=<rgba 0.5 0 1 1>}
  NamedRerouteDeclaration(None=@n44.B) {name=BM_Generic03, nodeColor=<rgba 0.5 0 1 1>}
  n45 = Saturate(None=Multiply(
    A=VectorParameter() {group=GenericBiomeColors, parameterName=GBC_Cliff_00, sortPriority=0, defaultValue=<rgba 1 1 1 1>}.RGB
    B=NamedRerouteUsage()
   ))
  n46 = VectorParameter() {group=GenericBiomeColors, parameterName=GBC_Cliff_01, sortPriority=10, defaultValue=<rgba 1 1 1 1>}
  n47 = Saturate(None=Multiply(A=@n46.RGB, B=NamedRerouteUsage()))
  n48 = NamedRerouteUsage()
  n49 = VectorParameter() {group=GenericBiomeColors, parameterName=GBC_Cliff_02, sortPriority=20, defaultValue=<rgba 1 1 1 1>}
  n50 = Multiply(A=@n49.RGB, B=NamedRerouteUsage())
  n51 = VectorParameter() {group=GenericBiomeColors, parameterName=GBC_Cliff_03, sortPriority=30, defaultValue=<rgba 1 1 1 1>}
  n52 = NamedRerouteUsage()
  n53 = LinearInterpolate(
   A=LinearInterpolate(A=LinearInterpolate(A=@n45, B=@n47, Alpha=@n48), B=Saturate(None=@n50), Alpha=NamedRerouteUsage())
   B=Saturate(None=Multiply(A=@n51.RGB, B=@n52))
   Alpha=NamedRerouteUsage()
  )
  n54 = VectorParameter() {group=GenericBiomeColors, parameterName=GBC_Cliff_04, sortPriority=40, defaultValue=<rgba 1 1 1 1>}
  n55 = Saturate(None=Multiply(A=@n54.RGB, B=NamedRerouteUsage()))
  n56 = NamedRerouteUsage()
  n57 = VectorParameter() {group=GenericBiomeColors, parameterName=GBC_Cliff_05, sortPriority=50, defaultValue=<rgba 1 1 1 1>}
  n58 = Multiply(A=@n57.RGB, B=NamedRerouteUsage())
  NamedRerouteDeclaration(None=LinearInterpolate(A=LinearInterpolate(A=@n53, B=@n55, Alpha=@n56), B=Saturate(None=@n58), Alpha=NamedRerouteUsage())) {name=OutputGeneric, nodeColor=<rgba 0.5 0 1 1>}
  NamedRerouteDeclaration(None=@n44.A) {name=BM_Generic04, nodeColor=<rgba 0.5 0 1 1>}
  n59 = MaterialFunctionCall(Float4=@n14) {materialFunction=&BreakOutFloat4Components}
  NamedRerouteDeclaration(None=@n59.R) {name=BM_Generic05, nodeColor=<rgba 0.5 0 1 1>}
  NamedRerouteDeclaration(None=@n59.G) {name=BM_Generic06, nodeColor=<rgba 0.5 0 1 1>}
  n60 = VectorParameter() {group=GenericBiomeColors, parameterName=GBC_Mobile_Cliff_00, sortPriority=0, defaultValue=<rgba 1 1 1 1>}
  n61 = Saturate(None=Multiply(A=@n60.RGB, B=NamedRerouteUsage()))
  n62 = VectorParameter() {group=GenericBiomeColors, parameterName=GBC_Mobile_Cliff_01, sortPriority=10, defaultValue=<rgba 1 1 1 1>}
  n63 = Saturate(None=Multiply(A=@n62.RGB, B=NamedRerouteUsage()))
  n64 = NamedRerouteUsage()
  n65 = VectorParameter() {group=GenericBiomeColors, parameterName=GBC_Mobile_Cliff_02, sortPriority=20, defaultValue=<rgba 1 1 1 1>}
  n66 = Multiply(A=@n65.RGB, B=NamedRerouteUsage())
  n67 = VectorParameter() {group=GenericBiomeColors, parameterName=GBC_Mobile_Cliff_03, sortPriority=30, defaultValue=<rgba 1 1 1 1>}
  n68 = NamedRerouteUsage()
  n69 = LinearInterpolate(
   A=LinearInterpolate(A=LinearInterpolate(A=@n61, B=@n63, Alpha=@n64), B=Saturate(None=@n66), Alpha=NamedRerouteUsage())
   B=Saturate(None=Multiply(A=@n67.RGB, B=@n68))
   Alpha=NamedRerouteUsage()
  )
  n70 = VectorParameter() {group=GenericBiomeColors, parameterName=GBC_Mobile_Cliff_04, sortPriority=40, defaultValue=<rgba 1 1 1 1>}
  n71 = Saturate(None=Multiply(A=@n70.RGB, B=NamedRerouteUsage()))
  n72 = NamedRerouteUsage()
  n73 = VectorParameter() {group=GenericBiomeColors, parameterName=GBC_Mobile_Cliff_05, sortPriority=50, defaultValue=<rgba 1 1 1 1>}
  n74 = Multiply(A=@n73.RGB, B=NamedRerouteUsage())
  NamedRerouteDeclaration(None=LinearInterpolate(A=LinearInterpolate(A=@n69, B=@n71, Alpha=@n72), B=Saturate(None=@n74), Alpha=NamedRerouteUsage())) {name=OutputGenericMobile, nodeColor=<rgba 0.5 0 1 1>}
  NamedRerouteDeclaration(None=@n15.08_IslandA) {name=BM_IslandA, nodeColor=<rgba 0.5 0 1 1>}
  NamedRerouteDeclaration(None=@n15.09_IslandB) {name=BM_IslandB, nodeColor=<rgba 0.5 0 1 1>}
  NamedRerouteDeclaration(None=@n15.06_NarrowFlea) {name=BM_NarrowFlea, nodeColor=<rgba 0.5 0 1 1>}
  NamedRerouteDeclaration(None=@n15.00_Chaparral) {name=BM_Empty, nodeColor=<rgba 0.5 0 1 1>}
  NamedRerouteDeclaration(None=@n15.01_RedWoods) {name=BM_RedWoods, nodeColor=<rgba 0.5 0 1 1>}
  NamedRerouteDeclaration(None=@n15.02_Archipelago) {name=BM_Archipelago, nodeColor=<rgba 0.5 0 1 1>}
  StaticSwitchParameter(True=@n1, False=ShadingPathSwitch(Default=@n3, Mobile=@n10)) {defaultValue=true, group=BiomeSettings, parameterName=UseBiomeColors}
 ]
}
