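# Landscape layer function that sets the grass base color per biome on the Hera terrain, with desktop and mobile outputs.
# Biome masks successively blend noise-varied color palettes, with NarrowFlea adding world-space stripes.
# Mobile uses cheaper per-biome tints.
# Static switches toggle biome colors, select generic palettes for non-BR maps, or swap parameter colors for a texture array, a switch marked to disable before ship.
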MaterialFunction /Hera/Environment/Terrain/Landscape/Materials/Layers/BiomeColor/MF_Hera_Grass_BiomeColor
graph {
 properties {libraryCategoriesText [FN-Landscape]}
 nodes [
  Constant4Vector() {constant=<rgba 0.3 0.0425 0.0167 1>}
  Constant4Vector() {constant=<rgba 0.75 1.5 0.8566 500>}
  Constant4Vector() {constant=<rgba 1.125 2.25 1.285 1>}
  Constant4Vector() {constant=<rgba 0.6 0.8 0.6986 1>}
  Constant4Vector() {constant=<rgba 0.25 0.0266 0.0147 1>}
  Constant4Vector() {constant=<rgba 0.66 1.1 0.7226 1>}
  Constant4Vector() {constant=<rgba 0.65 1.3 0.7424 1>}
  Constant4Vector() {constant=<rgba 1 0.9815 0.7 0.2>}
  Constant4Vector() {constant=<rgba 0.0845 0.1732 0.0065 1>}
  Constant4Vector() {constant=<rgba 0.0651 0.1137 0.0105 1>}
  Constant4Vector() {constant=<rgba 0.4 0.0607 0.0263 1>}
  Constant4Vector() {constant=<rgba 0.1263 0.144 0.0199 1>}
  Constant4Vector() {constant=<rgba 0.0503 0.13 0.0125 1>}
  Constant4Vector() {constant=<rgba 0.25 0.0188 0.0211 1>}
  Constant4Vector() {constant=<rgba 0.7 20 500 5000>}
  Constant() {r=0.7}
  Constant() {r=20}
  Constant() {desc="This offsets the ocean level used by the sand. Positive numbers go lower, negative numbers make it go higher.", r=-2200}
  Constant() {r=2500}
  n1 = NamedRerouteUsage()
  n2 = NamedRerouteUsage()
  FunctionOutput(None=SetMaterialAttributes!(
    MaterialAttributes=NamedRerouteUsage()
    BaseColor=StaticSwitchParameter(True=@n1, False=@n2) {defaultValue=true, group=BiomeSettings, parameterName=UseBiomeColors}
   ) {attributeSetTypes [#1]})
  n3 = NamedRerouteUsage()
  n4 = NamedRerouteUsage()
  FunctionOutput(None=SetMaterialAttributes(
    MaterialAttributes=NamedRerouteUsage()
    BaseColor=StaticSwitchParameter(True=@n3, False=@n4) {defaultValue=true, group=BiomeSettings, parameterName=UseBiomeColors}
   ) {attributeSetTypes [#1]}) {outputName=Result Mobile, sortPriority=1}
  n5 = VectorParameter() {group=BiomeColorsGrassIndividual, parameterName=GolfCourse_Base, sortPriority=-500, defaultValue=<rgba 1 0.9815 0.7 0.2>}
  n6 = VectorParameter() {group=BiomeColorsGrassIndividual, parameterName=GolfCourse_Rough, sortPriority=-450, defaultValue=<rgba 0.658 0.85 0.425 0.15>}
  n7 = NamedRerouteUsage()
  n8 = VectorParameter() {group=BiomeColorsGrassIndividual, parameterName=GolfCourse_Fairway, sortPriority=-400, defaultValue=<rgba 0.7309 1 0.5383 0.15>}
  n9 = VectorParameter() {group=BiomeColorsGrassIndividual, parameterName=GolfCourse_Fairway2, sortPriority=-350, defaultValue=<rgba 0.9065 1.2 0.6359 0.15>}
  n10 = MaterialFunctionCall!(Scale=ScalarParameter() {defaultValue=600, group=BiomeColorsGrassIndividual, parameterName=GolfCourse_Fairway_Scale, sortPriority=-290}) {materialFunction=/Game/Athena/Apollo/Environments/Landscape/Materials/Biome/CommonMaterials/MF_SimpleWorldUV}
  n11 = Multiply(A=ComponentMask(None=@n10.Result) {r=true}) {constB=-1}
  n12 = ComponentMask(None=@n10.Result) {g=true}
  n13 = LinearInterpolate(
   A=TextureSample(UVs=@n10.Result) {samplerSource=SSM_Wrap_WorldGroupSettings, texture=/Game/Creative/Devices/Generic/Materials/Textures/T_CP_Device_Stripes_01_D}.R
   B=TextureSample(UVs=MaterialFunctionCall(X=@n11, Y=@n12) {materialFunction=&MakeFloat2}.Result) {samplerSource=SSM_Wrap_WorldGroupSettings, texture=&T_CP_Device_Stripes_01_D}.R
  ) {constAlpha=0.35}
  n14 = VectorParameter() {group=BiomeColorsGrassIndividual, parameterName=GolfCourse_PuttingGreen, sortPriority=-300, defaultValue=<rgba 1.3003 1.5 0.841 0.2>}
  n15 = LinearInterpolate(
   A=LinearInterpolate(A=LinearInterpolate(A=@n5.RGBA, B=@n6.RGBA, Alpha=@n7), B=LinearInterpolate(A=@n8.RGBA, B=@n9.RGBA, Alpha=@n13), Alpha=NamedRerouteUsage())
   B=@n14.RGBA
   Alpha=NamedRerouteUsage()
  )
  n16 = Reroute()
  n17 = Multiply(A=ComponentMask(None=@n15) {b=true, g=true, r=true}, B=Desaturation(None=@n16, Fraction=ComponentMask(None=@n15) {a=true}))
  n18 = Multiply(A=@n5.RGB, B=Desaturation(None=Multiply(A=@n16), Fraction=@n5.A))
  LinearInterpolate(B=StaticSwitchParameter(True=Saturate(None=@n17), False=Saturate(None=@n18)) {group=BiomeColorsGrassIndividual, parameterName=UseGolfCourseGrassDetail})
  n19 = VectorParameter() {group=BiomeColorsGrassIndividual, parameterName=Mobile_GolfCourse_Base, sortPriority=-500, defaultValue=<rgba 1.1 1.0494 0.4196 1>}
  n20 = VectorParameter() {group=BiomeColorsGrassIndividual, parameterName=Mobile_GolfCourse_Rough, sortPriority=-450, defaultValue=<rgba 0.7222 0.9 0.315 0.15>}
  n21 = LinearInterpolate(A=@n19.RGB, B=@n20.RGB, Alpha=NamedRerouteUsage())
  n22 = VectorParameter() {group=BiomeColorsGrassIndividual, parameterName=Mobile_GolfCourse_Fairway, sortPriority=-400, defaultValue=<rgba 0.96 1.25 0.4545 0.15>}
  n23 = NamedRerouteUsage()
  n24 = VectorParameter() {group=BiomeColorsGrassIndividual, parameterName=Mobile_GolfCourse_PuttingGreen, sortPriority=-300, defaultValue=<rgba 1.3344 1.6 0.6993 0.2>}
  n25 = Multiply(A=LinearInterpolate(A=LinearInterpolate(A=@n21, B=@n22.RGB, Alpha=@n23), B=@n24.RGB, Alpha=NamedRerouteUsage()), B=NamedRerouteUsage())
  LinearInterpolate(B=ComponentMask(None=Saturate(None=@n25)) {b=true, g=true, r=true})
  n26 = VectorParameter() {group=BiomeColorsGrassIndividual, parameterName=CP_CA, defaultValue=<rgba 0.1145 0.1562 0.0171 1>}
  n27 = LinearInterpolate(
   A=LinearInterpolate(
    A=@n26.RGB
    B=VectorParameter() {group=BiomeColorsGrassIndividual, parameterName=CP_CB, defaultValue=<rgba 0.0754 0.1042 0.0096 1>}.RGB
    Alpha=NamedRerouteUsage()
   )
   B=Desaturation(None=@n26.RGB, Fraction=NamedRerouteUsage())
   Alpha=NamedRerouteUsage()
  )
  n28 = VectorParameter() {group=BiomeColorsGrassIndividual, parameterName=CP_CC, defaultValue=<rgba 0.1263 0.144 0.0199 1>}
  n29 = NamedRerouteUsage()
  LinearInterpolate(
   B=LinearInterpolate(
    A=VectorParameter() {group=BiomeColorsGrassIndividual, parameterName=CP_CD, defaultValue=<rgba 0.0723 0.1146 0.011 1>}.RGB
    B=LinearInterpolate(A=@n27, B=@n28.RGB, Alpha=@n29)
    Alpha=NamedRerouteUsage()
   )
   Alpha=NamedRerouteUsage()
  )
  n30 = VectorParameter() {
   parameterName=GoldenFallSettings
   channelNames {a=HeightDivide, b=HeightAdd, g=InputDiffusePower, r=HueShift}
   defaultValue=<rgba 0 1.5 3500 500>
  }
  n31 = Multiply()
  n32 = VectorParameter() {parameterName=GoldenFallLow_Color, channelNames {a=Desaturate, b="", g="", r=""}, defaultValue=<rgba 1 1 1 1>}
  n33 = Desaturation(
   None=MaterialFunctionCall(Hue Shift Percentage=@n30.HueShift, Texture=Power(Base=@n31, Exp=@n30.InputDiffusePower) {constExponent=16}) {materialFunction=&HueShift}.Result
   Fraction=@n32.Desaturate
  )
  n34 = ComponentMask(None=@n32.RGB) {b=true, g=true, r=true}
  n35 = Power(Base=ComponentMask(None=VertexNormalWS()) {b=true}, Exp=ScalarParameter() {defaultValue=5, parameterName=GoldenFall_SlopePower}) {constExponent=16}
  n36 = OneMinus(None=NamedRerouteUsage())
  LinearInterpolate(A=Saturate(None=Multiply(A=@n33, B=@n34)), B=@n31, Alpha=Saturate(None=Add(A=@n35, B=@n36)))
  n37 = Multiply()
  n38 = MaterialFunctionCall(Float3=@n37) {materialFunction=&BreakOutFloat3Components}
  n39 = OneMinus(None=MaterialFunctionCall(X=OneMinus(None=@n38.R), Y=@n38.G, Z=@n38.G) {materialFunction=&MakeFloat3}.Result)
  n40 = Constant() {r=20}
  n41 = Constant4Vector() {constant=<rgba 0.13 0.1283 0.013 0.5>}
  n42 = Desaturation(
   None=MaterialFunctionCall(Hue Shift Percentage=Constant() {r=0.7}, Texture=Power(Base=@n39, Exp=@n40) {constExponent=16}) {materialFunction=&HueShift}.Result
   Fraction=@n41.A
  )
  n43 = ComponentMask(None=@n41.RGBA) {b=true, g=true, r=true}
  n44 = TruncateLWC!(None=WorldPosition().Z)
  n45 = Constant() {desc="This offsets the ocean level used by the sand. Positive numbers go lower, negative numbers make it go higher.", r=-2200}
  n46 = Saturate(None=Divide(A=Add(A=@n44, B=@n45), B=Constant() {r=2500}))
  n47 = NamedRerouteUsage()
  LinearInterpolate(A=Saturate(None=Multiply(A=@n42, B=@n43)), B=@n37, Alpha=OneMinus(None=Multiply(A=@n46, B=@n47)))
  n48 = Reroute()
  n49 = VectorParameter() {parameterName=DesertGrassSimpleNoise_Color_BC, defaultValue=<rgba 0.4917 0.6 0.303 0.6>}
  n50 = NamedRerouteUsage()
  n51 = NamedRerouteUsage()
  LinearInterpolate(A=@n48, B=Multiply(A=@n49.RGB, B=@n48), Alpha=Multiply(A=@n49.A, B=Multiply(A=@n50, B=@n51)))
  n52 = Divide(
   A=Add(
    A=@n44
    B=Constant() {desc="This offsets the ocean level used by the sand. Positive numbers go lower, negative numbers make it go higher.", r=-200}
   )
   B=Constant() {r=1000}
  )
  n53 = Saturate(None=Add(A=Saturate(None=@n52), B=Constant() {r=0.58}))
  n54 = OneMinus(None=NamedRerouteUsage())
  Multiply(A=Saturate(None=Add(A=@n53, B=@n54)), B=Reroute())
  n55 = FunctionInput() {bUsePreviewValueAsDefault=true, inputName=TerrainMasks_QuadA, inputType=FunctionInput_Vector4, sortPriority=10}
  n56 = FunctionInput() {bUsePreviewValueAsDefault=true, inputName=TerrainMasks_QuadB, inputType=FunctionInput_Vector4, sortPriority=20}
  n57 = MaterialFunctionCall!(
   TerrainMasks_QuadA=@n55
   TerrainMasks_QuadB=@n56
   TerrainMasks_QuadC=FunctionInput() {bUsePreviewValueAsDefault=true, inputName=TerrainMasks_QuadC, inputType=FunctionInput_Vector4, sortPriority=30}
   TerrainMasks_QuadD=FunctionInput() {bUsePreviewValueAsDefault=true, inputName=TerrainMasks_QuadD, inputType=FunctionInput_Vector4, sortPriority=40}
  ) {materialFunction=/BRCoreMaterials/Hera/Terrain/BiomeMask/MF_HeraBiomeMasks_Isolated}
  NamedRerouteDeclaration(None=@n57.00_Chaparral) {name=BM_Chaparral, nodeColor=<rgba 0.5 0 1 1>}
  NamedRerouteDeclaration(None=@n57.01_RedWoods) {name=BM_RedWoods, nodeColor=<rgba 0.5 0 1 1>}
  n58 = VectorParameter() {group=BiomeColorsGrassIndividual, parameterName=RW_CA, defaultValue=<rgba 0.0654 0.1 0.025 1>}
  n59 = LinearInterpolate(
   A=LinearInterpolate(
    A=@n58.RGB
    B=VectorParameter() {group=BiomeColorsGrassIndividual, parameterName=RW_CB, defaultValue=<rgba 0.0743 0.15 0.0452 1>}.RGB
    Alpha=NamedRerouteUsage()
   )
   B=Desaturation(None=@n58.RGB, Fraction=NamedRerouteUsage())
   Alpha=NamedRerouteUsage()
  )
  n60 = VectorParameter() {group=BiomeColorsGrassIndividual, parameterName=RW_CC, defaultValue=<rgba 0.0922 0.1562 0.0525 1>}
  n61 = NamedRerouteUsage()
  n62 = LinearInterpolate(
   A=NamedRerouteUsage()
   B=LinearInterpolate(
    A=VectorParameter() {group=BiomeColorsGrassIndividual, parameterName=RW_CD, defaultValue=<rgba 0.0811 0.1458 0.041 1>}.RGB
    B=LinearInterpolate(A=@n59, B=@n60.RGB, Alpha=@n61)
    Alpha=NamedRerouteUsage()
   )
   Alpha=NamedRerouteUsage()
  )
  n63 = VectorParameter() {group=BiomeColorsGrassIndividual, parameterName=AP_CA, defaultValue=<rgba 0.0292 0.07 0.0031 1>}
  n64 = LinearInterpolate(
   A=@n63.RGB
   B=VectorParameter() {group=BiomeColorsGrassIndividual, parameterName=AP_CB, defaultValue=<rgba 0.0511 0.12 0.0121 1>}.RGB
   Alpha=NamedRerouteUsage()
  )
  n65 = Desaturation(None=@n63.RGB, Fraction=NamedRerouteUsage())
  n66 = NamedRerouteUsage()
  n67 = LinearInterpolate(
   A=VectorParameter() {group=BiomeColorsGrassIndividual, parameterName=AP_CD, defaultValue=<rgba 0.0613 0.11 0.0077 1>}.RGB
   B=LinearInterpolate(
    A=LinearInterpolate(A=@n64, B=@n65, Alpha=@n66)
    B=VectorParameter() {group=BiomeColorsGrassIndividual, parameterName=AP_CC, defaultValue=<rgba 0.0107 0.03 0.0042 1>}.RGB
    Alpha=NamedRerouteUsage()
   )
   Alpha=NamedRerouteUsage()
  )
  n68 = NamedRerouteUsage()
  n69 = VectorParameter() {group=BiomeColorsGrassIndividual, parameterName=DT_CD, defaultValue=<rgba 0.1026 0.125 0.0245 1>}
  n70 = VectorParameter() {group=BiomeColorsGrassIndividual, parameterName=DT_CA, defaultValue=<rgba 0.1304 0.15 0.0437 1>}
  n71 = VectorParameter() {group=BiomeColorsGrassIndividual, parameterName=DT_CB, defaultValue=<rgba 0.1239 0.1562 0.0251 1>}
  n72 = NamedRerouteUsage()
  n73 = NamedRerouteUsage()
  n74 = LinearInterpolate(
   A=LinearInterpolate(A=LinearInterpolate(A=@n70.RGB, B=@n71.RGB, Alpha=@n72), B=Desaturation(None=@n70.RGB, Fraction=@n73), Alpha=NamedRerouteUsage())
   B=VectorParameter() {group=BiomeColorsGrassIndividual, parameterName=DT_CC, defaultValue=<rgba 0.0884 0.1198 0.0237 1>}.RGB
   Alpha=NamedRerouteUsage()
  )
  n75 = NamedRerouteUsage()
  n76 = VectorParameter() {group=BiomeColorsGrassIndividual, parameterName=GF_CA, defaultValue=<rgba 0.4 0.1734 0.02 1>}
  n77 = LinearInterpolate(
   A=LinearInterpolate(
    A=@n76.RGB
    B=VectorParameter() {group=BiomeColorsGrassIndividual, parameterName=GF_CB, defaultValue=<rgba 0.3 0.1384 0.0353 1>}.RGB
    Alpha=NamedRerouteUsage()
   )
   B=Desaturation(None=@n76.RGB, Fraction=NamedRerouteUsage())
   Alpha=NamedRerouteUsage()
  )
  n78 = VectorParameter() {group=BiomeColorsGrassIndividual, parameterName=GF_CC, defaultValue=<rgba 0.45 0.219 0.0225 1>}
  n79 = NamedRerouteUsage()
  n80 = LinearInterpolate(
   A=LinearInterpolate(A=LinearInterpolate(A=@n62, B=@n67, Alpha=@n68), B=LinearInterpolate(A=@n69.RGB, B=@n74, Alpha=@n75), Alpha=NamedRerouteUsage())
   B=LinearInterpolate(
    A=VectorParameter() {group=BiomeColorsGrassIndividual, parameterName=GF_CD, defaultValue=<rgba 0.3 0.122 0.015 1>}.RGB
    B=LinearInterpolate(A=@n77, B=@n78.RGB, Alpha=@n79)
    Alpha=NamedRerouteUsage()
   )
   Alpha=NamedRerouteUsage()
  )
  n81 = VectorParameter() {group=BiomeColorsGrassIndividual, parameterName=GF2_CA, defaultValue=<rgba 0.2 0.0923 0.0236 1>}
  n82 = LinearInterpolate(
   A=@n81.RGB
   B=VectorParameter() {group=BiomeColorsGrassIndividual, parameterName=GF2_CB, defaultValue=<rgba 0.15 0.0708 0.015 1>}.RGB
   Alpha=NamedRerouteUsage()
  )
  n83 = Desaturation(None=@n81.RGB, Fraction=NamedRerouteUsage())
  n84 = NamedRerouteUsage()
  n85 = LinearInterpolate(
   A=VectorParameter() {group=BiomeColorsGrassIndividual, parameterName=GF2_CD, defaultValue=<rgba 0.15 0.0742 0.0225 1>}.RGB
   B=LinearInterpolate(
    A=LinearInterpolate(A=@n82, B=@n83, Alpha=@n84)
    B=VectorParameter() {group=BiomeColorsGrassIndividual, parameterName=GF2_CC, defaultValue=<rgba 0.112 0.059 0.0224 1>}.RGB
    Alpha=NamedRerouteUsage()
   )
   Alpha=NamedRerouteUsage()
  )
  n86 = NamedRerouteUsage()
  n87 = VectorParameter() {group=BiomeColorsGrassIndividual, parameterName=NF_CD, defaultValue=<rgba 0.0824 0.224 0.0278 1>}
  n88 = VectorParameter() {group=BiomeColorsGrassIndividual, parameterName=NF_CA, defaultValue=<rgba 0.0849 0.3073 0.02 1>}
  n89 = VectorParameter() {group=BiomeColorsGrassIndividual, parameterName=NF_CB, defaultValue=<rgba 0.1024 0.2812 0.0335 1>}
  n90 = NamedRerouteUsage()
  n91 = NamedRerouteUsage()
  n92 = LinearInterpolate(
   A=LinearInterpolate(A=LinearInterpolate(A=@n88.RGB, B=@n89.RGB, Alpha=@n90), B=Desaturation(None=@n88.RGB, Fraction=@n91), Alpha=NamedRerouteUsage())
   B=VectorParameter() {group=BiomeColorsGrassIndividual, parameterName=NF_CC, defaultValue=<rgba 0.1291 0.2917 0.0478 1>}.RGB
   Alpha=NamedRerouteUsage()
  )
  n93 = NamedRerouteUsage()
  n94 = NamedRerouteUsage()
  n95 = VectorParameter() {group=BiomeColorsGrassIndividual, parameterName=AM_CA, defaultValue=<rgba 0.35 0.0496 0.0195 1>}
  n96 = LinearInterpolate(
   A=LinearInterpolate(
    A=@n95.RGB
    B=VectorParameter() {group=BiomeColorsGrassIndividual, parameterName=AM_CB, defaultValue=<rgba 0.2 0.0386 0.03 1>}.RGB
    Alpha=NamedRerouteUsage()
   )
   B=Desaturation(None=@n95.RGB, Fraction=NamedRerouteUsage())
   Alpha=NamedRerouteUsage()
  )
  n97 = VectorParameter() {group=BiomeColorsGrassIndividual, parameterName=AM_CC, defaultValue=<rgba 0.6 0.0934 0.042 1>}
  n98 = NamedRerouteUsage()
  n99 = LinearInterpolate(
   A=LinearInterpolate(A=LinearInterpolate(A=@n80, B=@n85, Alpha=@n86), B=LinearInterpolate(A=@n87.RGB, B=@n92, Alpha=@n93), Alpha=Saturate(None=@n94))
   B=LinearInterpolate(
    A=VectorParameter() {group=BiomeColorsGrassIndividual, parameterName=AM_CD, defaultValue=<rgba 0.25 0.0312 0.0335 1>}.RGB
    B=LinearInterpolate(A=@n96, B=@n97.RGB, Alpha=@n98)
    Alpha=NamedRerouteUsage()
   )
   Alpha=NamedRerouteUsage()
  )
  n100 = VectorParameter() {group=BiomeColorsGrassIndividual, parameterName=IA_CA, defaultValue=<rgba 0.3 0.2318 0.0012 1>}
  n101 = LinearInterpolate(
   A=@n100.RGB
   B=VectorParameter() {group=BiomeColorsGrassIndividual, parameterName=IA_CB, defaultValue=<rgba 0.2 0.1315 0 1>}.RGB
   Alpha=NamedRerouteUsage()
  )
  n102 = Desaturation(None=@n100.RGB, Fraction=NamedRerouteUsage())
  n103 = NamedRerouteUsage()
  n104 = LinearInterpolate(
   A=VectorParameter() {group=BiomeColorsGrassIndividual, parameterName=IA_CD, defaultValue=<rgba 0.2 0.1305 0.0073 1>}.RGB
   B=LinearInterpolate(
    A=LinearInterpolate(A=@n101, B=@n102, Alpha=@n103)
    B=VectorParameter() {group=BiomeColorsGrassIndividual, parameterName=IA_CC, defaultValue=<rgba 0.25 0.1829 0.025 1>}.RGB
    Alpha=NamedRerouteUsage()
   )
   Alpha=NamedRerouteUsage()
  )
  n105 = NamedRerouteUsage()
  n106 = VectorParameter() {group=BiomeColorsGrassIndividual, parameterName=IB_CD, defaultValue=<rgba 0.15 0.0375 0.0231 1>}
  n107 = VectorParameter() {group=BiomeColorsGrassIndividual, parameterName=IB_CA, defaultValue=<rgba 0.2 0.0455 0.0151 1>}
  n108 = VectorParameter() {group=BiomeColorsGrassIndividual, parameterName=IB_CB, defaultValue=<rgba 0.15 0.0272 0.0148 1>}
  n109 = NamedRerouteUsage()
  n110 = NamedRerouteUsage()
  n111 = LinearInterpolate(
   A=LinearInterpolate(A=LinearInterpolate(A=@n107.RGB, B=@n108.RGB, Alpha=@n109), B=Desaturation(None=@n107.RGB, Fraction=@n110), Alpha=NamedRerouteUsage())
   B=VectorParameter() {group=BiomeColorsGrassIndividual, parameterName=IB_CC, defaultValue=<rgba 0.112 0.0288 0.0149 1>}.RGB
   Alpha=NamedRerouteUsage()
  )
  n112 = NamedRerouteUsage()
  n113 = VectorParameter() {group=BiomeColorsGrassIndividual, parameterName=IC_CA, defaultValue=<rgba 0.35 0.1657 0.35 1>}
  n114 = LinearInterpolate(
   A=LinearInterpolate(
    A=@n113.RGB
    B=VectorParameter() {group=BiomeColorsGrassIndividual, parameterName=IC_CB, defaultValue=<rgba 0.5 0.2368 0.5 1>}.RGB
    Alpha=NamedRerouteUsage()
   )
   B=Desaturation(None=@n113.RGB, Fraction=NamedRerouteUsage())
   Alpha=NamedRerouteUsage()
  )
  n115 = VectorParameter() {group=BiomeColorsGrassIndividual, parameterName=IC_CC, defaultValue=<rgba 0.25 0.1184 0.25 1>}
  n116 = NamedRerouteUsage()
  n117 = LinearInterpolate(
   A=LinearInterpolate(A=LinearInterpolate(A=@n99, B=@n104, Alpha=@n105), B=LinearInterpolate(A=@n106.RGB, B=@n111, Alpha=@n112), Alpha=NamedRerouteUsage())
   B=LinearInterpolate(
    A=VectorParameter() {group=BiomeColorsGrassIndividual, parameterName=IC_CD, defaultValue=<rgba 0.35 0.1657 0.35 1>}.RGB
    B=LinearInterpolate(A=@n114, B=@n115.RGB, Alpha=@n116)
    Alpha=NamedRerouteUsage()
   )
   Alpha=NamedRerouteUsage()
  )
  n118 = MaterialFunctionCall!(UseLandscapeCoords=StaticBoolParameter() {parameterName=UseLandscapeCoords}, "UseProjection(translucent)") {materialFunction=/BRCoreMaterials/Hera/Terrain/BiomeMask/MF_Hera_BiomeMasks}
  n119 = AppendVector(A=@n118.UVs, B=Constant() {r=3})
  n120 = TextureObjectParameter() {group=BiomeSettings, parameterName=BiomeGrassColorTextureArray, texture=/BRCoreMaterials/Hera/Terrain/BiomeMask/T_HeraBiomeGrass_Array}
  n121 = Constant()
  n122 = NamedRerouteDeclaration(None=TextureSample(UVs=AppendVector(A=@n118.UVs, B=@n121), Tex=@n120) {samplerSource=SSM_Wrap_WorldGroupSettings}.RGB) {name=BiomeGrassColorTexture01, nodeColor=<rgba 1 0 1 1>}
  n123 = AppendVector(A=@n118.UVs, B=Constant() {r=1})
  n124 = LinearInterpolate(
   A=@n122
   B=NamedRerouteDeclaration(None=TextureSample(UVs=@n123, Tex=@n120) {samplerSource=SSM_Wrap_WorldGroupSettings}.RGB) {name=BiomeGrassColorTexture02, nodeColor=<rgba 1 0 1 1>}
   Alpha=NamedRerouteUsage()
  )
  n125 = Desaturation(None=@n122, Fraction=NamedRerouteUsage())
  n126 = NamedRerouteUsage()
  n127 = TextureSample(UVs=AppendVector(A=@n118.UVs, B=Constant() {r=2}), Tex=@n120) {samplerSource=SSM_Wrap_WorldGroupSettings}
  n128 = LinearInterpolate(
   A=NamedRerouteDeclaration(None=TextureSample(UVs=@n119, Tex=@n120) {samplerSource=SSM_Wrap_WorldGroupSettings}.RGB) {name=BiomeGrassColorTexture04, nodeColor=<rgba 1 0 1 1>}
   B=LinearInterpolate(
    A=LinearInterpolate(A=@n124, B=@n125, Alpha=@n126)
    B=NamedRerouteDeclaration(None=@n127.RGB) {name=BiomeGrassColorTexture03, nodeColor=<rgba 1 0 1 1>}
    Alpha=NamedRerouteUsage()
   )
   Alpha=NamedRerouteUsage()
  )
  n129 = Multiply(A=NamedRerouteUsage(), B=StaticSwitch(True=@n117, False=@n128) {defaultValue=true, desc=Set this to false before ship})
  n130 = VectorParameter() {parameterName=NarrowFleaVariation_Color, channelNames {a=Desaturate, b="", g="", r=""}, defaultValue=<rgba 0.75 0.525 0.7495 1>}
  n131 = Multiply(A=@n130.RGB, B=@n129)
  n132 = Multiply(A=NamedRerouteUsage(), B=NamedRerouteUsage())
  NamedRerouteDeclaration(None=StaticSwitchParameter(True=NamedRerouteUsage(), False=LinearInterpolate(A=@n129, B=@n131, Alpha=@n132)) {desc=Used in maps outside of BR, group=GenericBiomeColors, parameterName=UseGenericBiome_ColorLerpChain}) {name=OutputBasecolor, nodeColor=<rgba 0.5 0 1 1>}
  n133 = VectorParameter() {group=BiomeColorsGrassIndividual, parameterName=GL_CA, defaultValue=<rgba 0.0845 0.1732 0.0065 1>}
  n134 = Constant() {r=0.14}
  n135 = LinearInterpolate(
   A=LinearInterpolate(
    A=@n133.RGB
    B=VectorParameter() {group=BiomeColorsGrassIndividual, parameterName=GL_CB, defaultValue=<rgba 0.0651 0.1137 0.0105 1>}.RGB
    Alpha=NamedRerouteUsage()
   )
   B=Desaturation(None=@n133.RGB, Fraction=NamedRerouteDeclaration(None=@n134) {name=Noise02_Desat, nodeColor=<rgba 0.5 0 1 1>})
   Alpha=NamedRerouteUsage()
  )
  n136 = VectorParameter() {group=BiomeColorsGrassIndividual, parameterName=GL_CC, defaultValue=<rgba 0.1263 0.144 0.0199 1>}
  n137 = NamedRerouteUsage()
  NamedRerouteDeclaration(None=LinearInterpolate(
    A=VectorParameter() {group=BiomeColorsGrassIndividual, parameterName=GL_CD, defaultValue=<rgba 0.0503 0.13 0.0125 1>}.RGB
    B=LinearInterpolate(A=@n135, B=@n136.RGB, Alpha=@n137)
    Alpha=NamedRerouteUsage()
   )) {name=NoBiomeColor, nodeColor=<rgba 0.5 0 1 1>}
  n138 = FunctionInput() {bUsePreviewValueAsDefault=true, inputName=GrassNoise, inputType=FunctionInput_Vector4, sortPriority=50, previewValue=<0 0 0 0>}
  NamedRerouteDeclaration(None=@n138) {name=GrassNoiseVec4, nodeColor=<rgba 0.5 0 1 1>}
  NamedRerouteDeclaration(None=@n57.07_AxisMandate) {name=BM_AxisMandate, nodeColor=<rgba 0.5 0 1 1>}
  NamedRerouteDeclaration(None=@n57.04_GoldenFall) {name=BM_GoldenFall, nodeColor=<rgba 0.5 0 1 1>}
  n139 = GetMaterialAttributes(None=FunctionInput(Preview=SetMaterialAttributes!(MaterialAttributes)) {bUsePreviewValueAsDefault=true, inputName=Material, inputType=FunctionInput_MaterialAttributes}) {attributeGetTypes [#1, #2]}
  n140 = NamedRerouteDeclaration(None=@n139.BaseColor) {name=Input_BaseColor, nodeColor=<rgba 0.5 0 1 1>}
  NamedRerouteDeclaration(None=Desaturation(None=@n140, Fraction=Constant() {r=0.5})) {name=BaseColor_DesatHalf, nodeColor=<rgba 0.5 0 1 1>}
  NamedRerouteDeclaration(None=@n57.06_NarrowFlea) {name=BM_NarrowFlea, nodeColor=<rgba 0.5 0 1 1>}
  n141 = Constant()
  n142 = Constant() {r=1.5}
  n143 = MaterialFunctionCall!(Coordinates, R_01_Scale, R_02_Scale, R_03_Scale, R_04_Scale, G_01_Scale, G_02_Scale, G_03_Scale, G_04_Scale) {materialFunction=/Game/Athena/Helios/Environments/Landscape/Materials/MaterialFunctions_Utilities/MF_VariationMasks2}
  NamedRerouteDeclaration(None=Saturate(None=LinearInterpolate(A=@n141, B=@n142, Alpha=@n143.G_01))) {name=Noise01, nodeColor=<rgba 0.5 0 1 1>}
  n144 = Constant() {r=1.5}
  n145 = Constant() {r=-9}
  NamedRerouteDeclaration(None=Saturate(None=LinearInterpolate(A=@n144, B=@n145, Alpha=@n143.G_02))) {name=Noise02, nodeColor=<rgba 0.5 0 1 1>}
  n146 = Constant() {r=-0.25}
  n147 = Constant() {r=1.5}
  n148 = NamedRerouteDeclaration(None=Saturate(None=LinearInterpolate(A=@n146, B=@n147, Alpha=@n143.G_03))) {name=Noise03_PreMult, nodeColor=<rgba 0.5 0 1 1>}
  n149 = Constant() {r=0.65}
  NamedRerouteDeclaration(None=Saturate(None=Multiply(A=@n148, B=@n149))) {name=Noise03, nodeColor=<rgba 0.5 0 1 1>}
  NamedRerouteDeclaration(None=@n57.02_Archipelago) {name=BM_Archipelago, nodeColor=<rgba 0.5 0 1 1>}
  n150 = Constant() {r=4}
  NamedRerouteDeclaration(None=Saturate(None=Multiply(A=@n143.G_04, B=@n150))) {name=Noise04, nodeColor=<rgba 0.5 0 1 1>}
  NamedRerouteDeclaration(None=FunctionInput() {bUsePreviewValueAsDefault=true, inputName=VariationMasks Mobile, sortPriority=51}) {name=VariationMasks Mobile, nodeColor=<rgba 1 0 0.6328 1>}
  NamedRerouteDeclaration() {name=BM_GolfCourse_Rough, nodeColor=<rgba 0.5 0 1 1>}
  NamedRerouteDeclaration() {name=BM_GolfCourse_Fairway, nodeColor=<rgba 0.5 0 1 1>}
  NamedRerouteDeclaration() {name=BM_GolfCourse_PuttingGreen, nodeColor=<rgba 0.5 0 1 1>}
  NamedRerouteDeclaration(None=FunctionInput(Preview=Constant()) {bUsePreviewValueAsDefault=true, inputName=DesertGrass_SimpleNoise, inputType=FunctionInput_Scalar, sortPriority=1}) {name=DesertGrass_SimpleNoise, nodeColor=<rgba 0.5 0 1 1>}
  n151 = MaterialFunctionCall(Float4=@n55) {materialFunction=&BreakOutFloat4Components}
  NamedRerouteDeclaration(None=@n151.R) {name=BM_Generic01, nodeColor=<rgba 0.5 0 1 1>}
  NamedRerouteDeclaration(None=@n151.G) {name=BM_Generic02, nodeColor=<rgba 0.5 0 1 1>}
  NamedRerouteDeclaration(None=@n151.B) {name=BM_Generic03, nodeColor=<rgba 0.5 0 1 1>}
  NamedRerouteDeclaration(None=@n151.A) {name=BM_Generic04, nodeColor=<rgba 0.5 0 1 1>}
  n152 = MaterialFunctionCall(Float4=@n56) {materialFunction=&BreakOutFloat4Components}
  NamedRerouteDeclaration(None=@n152.R) {name=BM_Generic05, nodeColor=<rgba 0.5 0 1 1>}
  NamedRerouteDeclaration(None=@n152.G) {name=BM_Generic06, nodeColor=<rgba 0.5 0 1 1>}
  n153 = VectorParameter() {group=GenericBiomeColors, parameterName=GBC_Grass_01_A, defaultValue=<rgba 0.1145 0.1562 0.0171 1>}
  n154 = LinearInterpolate(
   A=@n153.RGB
   B=VectorParameter() {group=GenericBiomeColors, parameterName=GBC_Grass_01_B, defaultValue=<rgba 0.0754 0.1042 0.0096 1>}.RGB
   Alpha=NamedRerouteUsage()
  )
  n155 = Desaturation(None=@n153.RGB, Fraction=NamedRerouteUsage())
  n156 = NamedRerouteUsage()
  n157 = LinearInterpolate(
   A=VectorParameter() {group=GenericBiomeColors, parameterName=GBC_Grass_01_D, defaultValue=<rgba 0.0723 0.1146 0.011 1>}.RGB
   B=LinearInterpolate(
    A=LinearInterpolate(A=@n154, B=@n155, Alpha=@n156)
    B=VectorParameter() {group=GenericBiomeColors, parameterName=GBC_Grass_01_C, defaultValue=<rgba 0.1263 0.144 0.0199 1>}.RGB
    Alpha=NamedRerouteUsage()
   )
   Alpha=NamedRerouteUsage()
  )
  n158 = NamedRerouteUsage()
  n159 = VectorParameter() {group=GenericBiomeColors, parameterName=GBC_Grass_00_A, defaultValue=<rgba 0.1145 0.1562 0.0171 1>}
  n160 = LinearInterpolate(
   A=@n159.RGB
   B=VectorParameter() {group=GenericBiomeColors, parameterName=GBC_Grass_00_B, defaultValue=<rgba 0.0754 0.1042 0.0096 1>}.RGB
   Alpha=NamedRerouteUsage()
  )
  n161 = Desaturation(None=@n159.RGB, Fraction=NamedRerouteUsage())
  n162 = NamedRerouteUsage()
  n163 = LinearInterpolate(
   A=VectorParameter() {group=GenericBiomeColors, parameterName=GBC_Grass_00_D, defaultValue=<rgba 0.0723 0.1146 0.011 1>}.RGB
   B=LinearInterpolate(
    A=LinearInterpolate(A=@n160, B=@n161, Alpha=@n162)
    B=VectorParameter() {group=GenericBiomeColors, parameterName=GBC_Grass_00_C, defaultValue=<rgba 0.1263 0.144 0.0199 1>}.RGB
    Alpha=NamedRerouteUsage()
   )
   Alpha=NamedRerouteUsage()
  )
  n164 = VectorParameter() {group=GenericBiomeColors, parameterName=GBC_Grass_02_D, defaultValue=<rgba 0.0811 0.1458 0.041 1>}
  n165 = VectorParameter() {group=GenericBiomeColors, parameterName=GBC_Grass_02_A, defaultValue=<rgba 0.0654 0.1 0.025 1>}
  n166 = VectorParameter() {group=GenericBiomeColors, parameterName=GBC_Grass_02_B, defaultValue=<rgba 0.0743 0.15 0.0452 1>}
  n167 = NamedRerouteUsage()
  n168 = NamedRerouteUsage()
  n169 = LinearInterpolate(
   A=LinearInterpolate(A=LinearInterpolate(A=@n165.RGB, B=@n166.RGB, Alpha=@n167), B=Desaturation(None=@n165.RGB, Fraction=@n168), Alpha=NamedRerouteUsage())
   B=VectorParameter() {group=GenericBiomeColors, parameterName=GBC_Grass_02_C, defaultValue=<rgba 0.0922 0.1562 0.0525 1>}.RGB
   Alpha=NamedRerouteUsage()
  )
  n170 = NamedRerouteUsage()
  n171 = VectorParameter() {group=GenericBiomeColors, parameterName=GBC_Grass_03_A, defaultValue=<rgba 0.0292 0.07 0.0031 1>}
  n172 = LinearInterpolate(
   A=LinearInterpolate(
    A=@n171.RGB
    B=VectorParameter() {group=GenericBiomeColors, parameterName=GBC_Grass_03_B, defaultValue=<rgba 0.0511 0.12 0.0121 1>}.RGB
    Alpha=NamedRerouteUsage()
   )
   B=Desaturation(None=@n171.RGB, Fraction=NamedRerouteUsage())
   Alpha=NamedRerouteUsage()
  )
  n173 = VectorParameter() {group=GenericBiomeColors, parameterName=GBC_Grass_03_C, defaultValue=<rgba 0.0107 0.03 0.0042 1>}
  n174 = NamedRerouteUsage()
  n175 = LinearInterpolate(
   A=LinearInterpolate(A=LinearInterpolate(A=@n163, B=@n157, Alpha=@n158), B=LinearInterpolate(A=@n164.RGB, B=@n169, Alpha=@n170), Alpha=NamedRerouteUsage())
   B=LinearInterpolate(
    A=VectorParameter() {group=GenericBiomeColors, parameterName=GBC_Grass_03_D, defaultValue=<rgba 0.0613 0.11 0.0077 1>}.RGB
    B=LinearInterpolate(A=@n172, B=@n173.RGB, Alpha=@n174)
    Alpha=NamedRerouteUsage()
   )
   Alpha=NamedRerouteUsage()
  )
  n176 = VectorParameter() {group=GenericBiomeColors, parameterName=GBC_Grass_04_A, defaultValue=<rgba 0.1304 0.15 0.0437 1>}
  n177 = LinearInterpolate(
   A=@n176.RGB
   B=VectorParameter() {group=GenericBiomeColors, parameterName=GBC_Grass_04_B, defaultValue=<rgba 0.1239 0.1562 0.0251 1>}.RGB
   Alpha=NamedRerouteUsage()
  )
  n178 = Desaturation(None=@n176.RGB, Fraction=NamedRerouteUsage())
  n179 = NamedRerouteUsage()
  n180 = LinearInterpolate(
   A=VectorParameter() {group=GenericBiomeColors, parameterName=GBC_Grass_04_D, defaultValue=<rgba 0.1026 0.125 0.0245 1>}.RGB
   B=LinearInterpolate(
    A=LinearInterpolate(A=@n177, B=@n178, Alpha=@n179)
    B=VectorParameter() {group=GenericBiomeColors, parameterName=GBC_Grass_04_C, defaultValue=<rgba 0.0884 0.1198 0.0237 1>}.RGB
    Alpha=NamedRerouteUsage()
   )
   Alpha=NamedRerouteUsage()
  )
  n181 = NamedRerouteUsage()
  n182 = VectorParameter() {group=GenericBiomeColors, parameterName=GBC_Grass_05_D, defaultValue=<rgba 0.25 0.1163 0.025 1>}
  n183 = VectorParameter() {group=GenericBiomeColors, parameterName=GBC_Grass_05_A, defaultValue=<rgba 0.3 0.1536 0.03 1>}
  n184 = VectorParameter() {group=GenericBiomeColors, parameterName=GBC_Grass_05_B, defaultValue=<rgba 0.4 0.1933 0.0474 1>}
  n185 = NamedRerouteUsage()
  n186 = NamedRerouteUsage()
  n187 = LinearInterpolate(
   A=LinearInterpolate(A=LinearInterpolate(A=@n183.RGB, B=@n184.RGB, Alpha=@n185), B=Desaturation(None=@n183.RGB, Fraction=@n186), Alpha=NamedRerouteUsage())
   B=VectorParameter() {group=GenericBiomeColors, parameterName=GBC_Grass_05_C, defaultValue=<rgba 0.4 0.1269 0.0302 1>}.RGB
   Alpha=NamedRerouteUsage()
  )
  n188 = NamedRerouteUsage()
  NamedRerouteDeclaration(None=LinearInterpolate(A=LinearInterpolate(A=@n175, B=@n180, Alpha=@n181), B=LinearInterpolate(A=@n182.RGB, B=@n187, Alpha=@n188), Alpha=NamedRerouteUsage())) {name=OutputGeneric, nodeColor=<rgba 0.5 0 1 1>}
  NamedRerouteDeclaration(None=@n57.05_GoldenFall2) {name=BM_GoldenFall2, nodeColor=<rgba 0.5 0 1 1>}
  NamedRerouteDeclaration(None=@n57.03_Desert) {name=BM_Desert, nodeColor=<rgba 0.5 0 1 1>}
  n189 = Saturate(None=Multiply(
    A=NamedRerouteUsage()
    B=VectorParameter() {group=GenericBiomeColors, parameterName=GBC_Mobile_Grass_00, defaultValue=<rgba 1.4 0.98 1.16 1>}.RGB
   ))
  n190 = Saturate(None=Multiply(
    A=NamedRerouteUsage()
    B=VectorParameter() {group=GenericBiomeColors, parameterName=GBC_Mobile_Grass_01, defaultValue=<rgba 1.4 0.98 1.16 1>}.RGB
   ))
  n191 = NamedRerouteUsage()
  n192 = Multiply(
   A=NamedRerouteUsage()
   B=VectorParameter() {group=GenericBiomeColors, parameterName=GBC_Mobile_Grass_02, defaultValue=<rgba 1.4 0.98 1.16 1>}.RGB
  )
  n193 = NamedRerouteUsage()
  n194 = VectorParameter() {group=GenericBiomeColors, parameterName=GBC_Mobile_Grass_03, defaultValue=<rgba 1.4 0.98 1.16 1>}
  n195 = LinearInterpolate(
   A=LinearInterpolate(A=LinearInterpolate(A=@n189, B=@n190, Alpha=@n191), B=Saturate(None=@n192), Alpha=NamedRerouteUsage())
   B=Saturate(None=Multiply(A=@n193, B=@n194.RGB))
   Alpha=NamedRerouteUsage()
  )
  n196 = Saturate(None=Multiply(
    A=NamedRerouteUsage()
    B=VectorParameter() {group=GenericBiomeColors, parameterName=GBC_Mobile_Grass_04, defaultValue=<rgba 1.4 0.98 1.16 1>}.RGB
   ))
  n197 = NamedRerouteUsage()
  n198 = Multiply(
   A=NamedRerouteUsage()
   B=VectorParameter() {group=GenericBiomeColors, parameterName=GBC_Mobile_Grass_05, defaultValue=<rgba 1.4 0.98 1.16 1>}.RGB
  )
  NamedRerouteDeclaration(None=LinearInterpolate(A=LinearInterpolate(A=@n195, B=@n196, Alpha=@n197), B=Saturate(None=@n198), Alpha=NamedRerouteUsage())) {name=OutputGenericMobile, nodeColor=<rgba 0.5 0 1 1>}
  NamedRerouteDeclaration(None=@n57.08_IslandA) {name=BM_IslandA, nodeColor=<rgba 0.5 0 1 1>}
  NamedRerouteDeclaration(None=@n57.09_IslandB) {name=BM_IslandB, nodeColor=<rgba 0.5 0 1 1>}
  n199 = ScalarParameter() {defaultValue=0.4, parameterName=NarrowFleaLandscapeGrassStripeBWidth}
  n200 = Saturate(None=ScalarParameter() {defaultValue=0.9, parameterName=NarrowFleaLandscapeGrassStripeSharpness})
  n201 = Max(A=Multiply(A=OneMinus(None=@n200)) {constB=0.5}) {constB=0.0001}
  n202 = Constant2Vector() {g=1, r=1}
  n203 = DotProduct(
   A=WorldPosition().XY
   B=Divide(A=Normalize(VectorInput=@n202), B=ScalarParameter() {defaultValue=724.0773, parameterName=NF_StripeFrequency})
  )
  n204 = Frac(None=Add(A=@n203, B=ScalarParameter() {defaultValue=0.25, parameterName=NarrowFleaLandscapeGrassStripeBOffset}) {constB=0.25})
  n205 = SmoothStep(Min=Subtract(A=@n199, B=@n201), Max=Add(A=@n199, B=@n201), Value=Abs(None=ConstantBiasScale(None=@n204) {bias=-0.5, scale=2})) {constMax=0.45, constMin=0.35}
  n206 = LinearInterpolate(
   A=ScalarParameter() {defaultValue=0.94, parameterName=NarrowFleaLandscapeGrassStripeA}
   B=ScalarParameter() {defaultValue=1, parameterName=NarrowFleaLandscapeGrassStripeB}
   Alpha=@n205
  )
  n207 = LinearInterpolate(
   A=ScalarParameter() {defaultValue=0.75, parameterName=NarrowFleaLandscapeGrassStripeC}
   B=ScalarParameter() {defaultValue=0.8, parameterName=NarrowFleaLandscapeGrassStripeD}
   Alpha=@n205
  )
  n208 = ScalarParameter() {defaultValue=0.5, parameterName=NarrowFleaLandscapeGrassStripeAWidth}
  n209 = Frac(None=@n203)
  n210 = SmoothStep(Min=Subtract(A=@n208, B=@n201), Max=Add(A=@n208, B=@n201), Value=Abs(None=ConstantBiasScale(None=@n209) {bias=-0.5, scale=2})) {constMax=0.55, constMin=0.45}
  n211 = Subtract(A=NamedRerouteUsage())
  n212 = OneMinus(None=Saturate(None=Abs(None=@n211)))
  n213 = OneMinus(None=NamedRerouteUsage())
  NamedRerouteDeclaration(None=LinearInterpolate(B=LinearInterpolate(A=@n206, B=@n207, Alpha=@n210) {constA=0.8}, Alpha=Multiply(A=@n212, B=@n213)) {constA=1}) {name=NarrowFleaLandscapeStripes, nodeColor=<rgba 0.1719 1 0 1>}
  NamedRerouteDeclaration(None=@n57.10_IslandC) {name=BM_IslandC, nodeColor=<rgba 0.5 0 1 1>}
  NamedRerouteDeclaration(None=@n57.11_RemoveStripes) {name=BM_RemoveStripes, nodeColor=<rgba 0.5 0 1 1>}
  NamedRerouteDeclaration(None=@n57.12_NarrowFleaVariation) {name=BM_NarrowFleaVariation, nodeColor=<rgba 0.5 0 1 1>}
  NamedRerouteDeclaration(None=FunctionInput(Preview=Constant()) {bUsePreviewValueAsDefault=true, inputName=GrassEdgeMask, inputType=FunctionInput_Scalar, sortPriority=1}) {name=GrassEdgeMask, nodeColor=<rgba 0.5 0 1 1>}
  NamedRerouteDeclaration(None=Desaturation(None=@n140, Fraction=Constant() {r=1})) {name=BaseColorDesat, nodeColor=<rgba 0.5 0 1 1>}
  NamedRerouteDeclaration(None=@n139.MaterialAttributes) {name=InputMaterial, nodeColor=<rgba 0.5 0 1 1>}
  NamedRerouteDeclaration(None=@n139.OpacityMask) {name=Heightmap, nodeColor=<rgba 0.5 0 1 1>}
  n214 = NamedRerouteUsage()
  n215 = Multiply(
   A=NamedRerouteUsage()
   B=VectorParameter() {group=BiomeColors, parameterName=Mobile_GrassColor_Grasslands, defaultValue=<rgba 1.4 0.98 1.16 1>}.RGB
  )
  n216 = VectorParameter() {group=BiomeColors, parameterName=Mobile_GrassColor_RedWoods, defaultValue=<rgba 0.5551 0.5057 0.75 1>}
  n217 = NamedRerouteUsage()
  n218 = LinearInterpolate(
   A=NamedRerouteDeclaration(None=Saturate(None=@n215)) {name=Mobile_NoBiomeColor, nodeColor=<rgba 0.5 0 1 1>}
   B=Saturate(None=Multiply(A=@n216.RGB, B=@n217))
   Alpha=NamedRerouteUsage()
  )
  n219 = VectorParameter() {group=BiomeColors, parameterName=Mobile_GrassColor_Archipelago, defaultValue=<rgba 0.4849 0.6391 0.75 1>}
  n220 = Saturate(None=Multiply(A=@n219.RGB, B=NamedRerouteUsage()))
  n221 = NamedRerouteUsage()
  n222 = Multiply(
   A=VectorParameter() {group=BiomeColors, parameterName=Mobile_GrassColor_Desert, defaultValue=<rgba 1.1 1.0494 0.4196 1>}.RGB
   B=NamedRerouteUsage()
  )
  n223 = VectorParameter() {group=BiomeColors, parameterName=Mobile_GrassColor_GoldenFall, defaultValue=<rgba 2.5 1.1924 0.3068 1>}
  n224 = NamedRerouteUsage()
  n225 = LinearInterpolate(
   A=LinearInterpolate(A=LinearInterpolate(A=@n218, B=@n220, Alpha=@n221), B=Saturate(None=@n222), Alpha=NamedRerouteUsage())
   B=Saturate(None=Multiply(A=@n223.RGB, B=@n224))
   Alpha=NamedRerouteUsage()
  )
  n226 = Saturate(None=Multiply(
    A=VectorParameter() {group=BiomeColors, parameterName=Mobile_GrassColor_NarrowFlea, defaultValue=<rgba 0.5025 1.5 0.15 1>}.RGB
    B=NamedRerouteUsage()
   ))
  n227 = NamedRerouteUsage()
  n228 = VectorParameter() {group=BiomeColors, parameterName=Mobile_GrassColor_AxisMandate, defaultValue=<rgba 1.75 0.3186 0.175 1>}
  n229 = Multiply(A=@n228.RGB, B=NamedRerouteUsage())
  n230 = VectorParameter() {group=BiomeColors, parameterName=Mobile_GrassColor_IslandA, defaultValue=<rgba 1.75 1.4502 0.1251 1>}
  n231 = NamedRerouteUsage()
  n232 = LinearInterpolate(
   A=LinearInterpolate(A=LinearInterpolate(A=@n225, B=@n226, Alpha=@n227), B=Saturate(None=@n229), Alpha=NamedRerouteUsage())
   B=Saturate(None=Multiply(A=@n230.RGB, B=@n231))
   Alpha=NamedRerouteUsage()
  )
  n233 = Saturate(None=Multiply(
    A=VectorParameter() {group=BiomeColors, parameterName=Mobile_GrassColor_IslandB, defaultValue=<rgba 1.75 0.3186 0.175 1>}.RGB
    B=NamedRerouteUsage()
   ))
  n234 = NamedRerouteUsage()
  n235 = Multiply(
   A=VectorParameter() {group=BiomeColors, parameterName=Mobile_GrassColor_IslandC, defaultValue=<rgba 1.95 0.4409 1.8778 1>}.RGB
   B=NamedRerouteUsage()
  )
  n236 = Multiply(
   A=NamedRerouteUsage()
   B=LinearInterpolate(A=LinearInterpolate(A=@n232, B=@n233, Alpha=@n234), B=Saturate(None=@n235), Alpha=NamedRerouteUsage())
  )
  n237 = LinearInterpolate(A=@n236, B=Multiply(A=@n130.RGB, B=@n236), Alpha=@n132)
  NamedRerouteDeclaration(None=ComponentMask(None=StaticSwitchParameter(True=@n214, False=@n237) {desc=Used in maps outside of BR, group=GenericBiomeColors, parameterName=UseGenericBiome_ColorLerpChain}) {b=true, g=true, r=true}) {name=OutputBasecolorMobile, nodeColor=<rgba 0.5 0 1 1>}
  NamedRerouteUsage()
  NamedRerouteUsage()
  NamedRerouteUsage()
  NamedRerouteUsage()
  NamedRerouteUsage()
  NamedRerouteUsage()
  NamedRerouteUsage()
  NamedRerouteUsage()
  NamedRerouteUsage()
  NamedRerouteUsage()
  NamedRerouteUsage()
  NamedRerouteUsage()
  NamedRerouteUsage()
  NamedRerouteUsage()
  NamedRerouteUsage()
  n238 = MaterialFunctionCall(Float3=@n31) {materialFunction=&BreakOutFloat3Components}
  OneMinus(None=MaterialFunctionCall(X=OneMinus(None=@n238.R), Y=@n238.G, Z=@n238.G) {materialFunction=&MakeFloat3}.Result)
  n239 = TruncateLWC(None=WorldPosition() {worldPositionShaderOffset=WPT_ExcludeAllShaderOffsets}.Z)
  OneMinus(None=@n239)
  n240 = Divide(A=Add(A=@n30.HeightAdd, B=@n239), B=@n30.HeightDivide)
  n241 = NamedRerouteUsage()
  Saturate(None=Add(A=Saturate(None=@n240), B=OneMinus(None=@n241)))
  ScalarParameter() {defaultValue=0.14, group=Grass, parameterName=Noise02_Desaturation}
  StaticSwitchParameter(True=@n128, False=@n117) {
   desc=This will be set to true in the MF itself as the default before release. It may be set to false during development to make biome mask painting easier.
   group=BiomeSettings
   parameterName=UseBiomeColorArray
  }
  StaticSwitchParameter() {parameterName=TestingThisSwitchToFalseToMakeSureThisWorks}
 ]
}
